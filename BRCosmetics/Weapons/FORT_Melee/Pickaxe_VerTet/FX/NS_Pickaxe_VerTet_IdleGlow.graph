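# Idle glow Niagara system for the VerTet pickaxe cosmetic, built from one sprite emitter, CapsuleGlow.
# Four burst spawn groups, some partitioned, create persistent sprites that follow sockets through a socket reader, with per-group size, offset and rotated alignment.
# Color comes from a variant selector, scaled by time of day and platform, while alpha fades in on system age with separate front-end and gameplay curves.
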NiagaraSystem /BRCosmetics/Weapons/FORT_Melee/Pickaxe_VerTet/FX/NS_Pickaxe_VerTet_IdleGlow
dependencies {
 usedDataInterfaces [
  &NiagaraDataInterfacePlatformSet
  &NiagaraDataInterfaceArrayFloat3
  &NiagaraDataInterfaceCurve
  &NiagaraDataInterfaceSocketReader
 ]
 usedDynamicInputs [&FNTimeOfDayScalar]
 usedModules [
  &self:SystemSpawnScript.SystemScriptSource.SystemScriptGraph.NiagaraNodeAssignment_0.SetVariables_#1
  &SystemState
  &self:SystemSpawnScript.SystemScriptSource.SystemScriptGraph.NiagaraNodeAssignment_1.SetVariables_#2
  &FN_VariantManager
 ]
}
emitters {CapsuleGlow {
  emitterUpdateScript [
   &EmitterState(
    Inactive Response=::NewEnumerator1
    Life Cycle Mode=::NewEnumerator1
    Loop Behavior=::NewEnumerator1
    Loop Duration Mode=::NewEnumerator1
    UseLoopDelay=false
   )
   &SpawnBurst_Instantaneous(Spawn Count=2, Use Spawn Probability=false, UseLoopCountLimit=false) {module=SpawnBurst_Instantaneous001}
   &SpawnBurst_Instantaneous(Spawn Count=2, Spawn Group=1, Use Spawn Probability=false, UseLoopCountLimit=false)
   &SpawnBurst_Instantaneous(Spawn Group=2, Use Spawn Probability=false, UseLoopCountLimit=false) {module=SpawnBurst_Instantaneous002}
   &SpawnBurst_Instantaneous(Spawn Group=3, Use Spawn Probability=false, UseLoopCountLimit=false) {module=SpawnBurst_Instantaneous003}
  ]
  particleSpawnScript [
   &PartitionParticles()
   /SetVariables(Particles.SpawnGroup=$Emitter.SpawnGroup) {isSetParameters=true}
   &InitializeParticle(
    Color=<rgba 5 5 5 1>
    Sprite Size=&MergeHelper_Vector2Switch(
     0=&MergeHelper_Vector2Switch(0=<3 9>, 1=<3 6>, 2=?, 3=?, 4=?, Index=$Particles.PartitionParticles.Partition, Num=::NewEnumerator0)
     1=&MergeHelper_Vector2Switch(0=<6 18>, 1=<6 8>, 2=?, 3=?, 4=?, Index=$Particles.PartitionParticles.Partition, Num=::NewEnumerator0)
     2=<5 7>
     3=<4 4>
     4=?
     Index=$Emitter.SpawnGroup
     Num=::NewEnumerator2
    )
    Sprite Size Mode=::NewEnumerator4
    UsePositionOffset=false
    Write Material Random=false
   )
   &CameraOffset(Camera Offset Amount=&MergeHelper_FloatSwitch(
     0=2
     1=&FN_PickFloatDependOnPlatform(TargetPlatformActive=2, TargetPlatformNoActive=3.5)
     2=2
     3=2
     Index=$Particles.SpawnGroup
     Num=::NewEnumerator2
    ))
  ]
  particleUpdateScript [
   &ParticleState(Kill Particles When Lifetime Has Elapsed=false)
   &DynamicMaterialParameters(
    Index 0 Param 1=1
    Index 0 Param 2=10
    Index 0 Param 3=1
    Module.Index0Float0Write=true
    Module.Index0Float1Write=true
    Module.Index0Float2Write=true
    Module.Index0Float3Write=false
   )
   &SocketLocation(
    Orientation Sampling=::NewEnumerator1
    SamplePosOffset=true
    Sampled Position Offset=&MergeHelper_VectorSwitch(
     0=&SelectVectorFromArray(
      Direct Array Index=$Particles.PartitionParticles.Partition
      Vector Selection Array=DataInterface() {propertyValues {FloatData [<-0.566 -1.996 14.45>, <-0.593 -3.623 19.521>]}}
     )
     1=&SelectVectorFromArray(
      Direct Array Index=$Particles.PartitionParticles.Partition
      Vector Selection Array=DataInterface() {propertyValues {FloatData [<-0.566 -8.6631 25.151>, <-0.566 -16.142 23.965>]}}
     )
     2=<-0.501 -0.109 -9.463>
     3=<-0.566 -1.0048 27.5198>
     Index=$Particles.SpawnGroup
     Num=::NewEnumerator2
    )
    Socket Reader=$System.Socket Reader
    Socket Sampling Mode=::NewEnumerator3
    VelocitySpeedLimit=false
   )
   &ScaleColor(
    Scale Alpha=&FN_FrontEndFloatModifier(
     Gameplay=&FloatFromCurve(FloatCurve=DataInterface() {propertyValues {Curve {keys [<at 0 0 RCIM_Cubic>,<at 0.65 1 RCIM_Cubic>]}}}, CurveIndex=$Engine.System.Age)
     Front End=&FloatFromCurve(FloatCurve=DataInterface() {propertyValues {Curve {keys [<at 0.35 0 RCIM_Cubic>,<at 1.1 1 RCIM_Cubic>]}}}, CurveIndex=$Engine.System.Age)
    )
    Scale RGB=&MakeVectorFromLinearColorRGB(Color=&Multiply_LinearColorByFloat(
      Linear Color=&VariantColorSelector(
       Variant Amount=4
       Original Color=&MergeHelper_ColorSwitch(
        0=&MergeHelper_ColorSwitch(0=<rgba 0.0409 0 0.3663 1>, 1=<rgba 0.1812 0.0009 0.5583 1>, 2=?, 3=?, Index=$Particles.PartitionParticles.Partition, Num=::NewEnumerator0)
        1=&MergeHelper_ColorSwitch(0=<rgba 0.2159 0.0044 0.5711 1>, 1=<rgba 0.0765 0.07 0.7 1>, 2=?, 3=?, Index=$Particles.PartitionParticles.Partition, Num=::NewEnumerator0)
        2=&LerpLinearColors(StartColor=<rgba 0.2238 0 0.924 1>, EndColor=<rgba 0.0687 0 0.794 1>, LerpFactor=0.5)
        3=<rgba 0.1413 0 0.5395 1>
        Index=$Particles.SpawnGroup
        Num=::NewEnumerator2
       )
       Variant 1 Color=&MergeHelper_ColorSwitch(
        0=<rgba 0.0504 0.0535 0.065 1>
        1=<rgba 0.1163 0.1234 0.15 1>
        2=<rgba 0.1046 0.1111 0.135 1>
        3=<rgba 0.1163 0.1234 0.15 1>
        Index=$Particles.SpawnGroup
        Num=::NewEnumerator2
       )
       Variant 2 Color=&MergeHelper_ColorSwitch(
        0=&MergeHelper_ColorSwitch(0=<rgba 0.15 0.0921 0.011 1>, 1=<rgba 0.15 0.1253 0.0488 1>, 2=?, 3=?, Index=$Particles.PartitionParticles.Partition, Num=::NewEnumerator0)
        1=&MergeHelper_ColorSwitch(0=<rgba 0.15 0.1253 0.0488 1>, 1=<rgba 0.15 0.0921 0.011 1>, 2=?, 3=?, Index=$Particles.PartitionParticles.Partition, Num=::NewEnumerator0)
        2=<rgba 0.16 0.1336 0.0521 1>
        3=<rgba 0.16 0.1336 0.0521 1>
        Index=$Particles.SpawnGroup
        Num=::NewEnumerator2
       )
       Variant 3 Color=&MergeHelper_ColorSwitch(
        0=&MergeHelper_ColorSwitch(0=<rgba 0.0065 0.1095 0.15 1>, 1=<rgba 0.0087 0.146 0.2 1>, 2=?, 3=?, Index=$Particles.PartitionParticles.Partition, Num=::NewEnumerator0)
        1=&MergeHelper_ColorSwitch(0=<rgba 0.0215 0.1537 0.215 1>, 1=<rgba 0.013 0.219 0.3 1>, 2=?, 3=?, Index=$Particles.PartitionParticles.Partition, Num=::NewEnumerator0)
        2=<rgba 0.013 0.219 0.3 1>
        3=<rgba 0.0225 0.1678 0.225 1>
        Index=$Particles.SpawnGroup
        Num=::NewEnumerator2
       )
       Variant 4 Color=&MergeHelper_ColorSwitch(
        0=&MergeHelper_ColorSwitch(0=<rgba 0.115 0 0.088 1>, 1=<rgba 0.2 0.0386 0.1816 1>, 2=?, 3=?, Index=$Particles.PartitionParticles.Partition, Num=::NewEnumerator0)
        1=&MergeHelper_ColorSwitch(0=<rgba 0.2 0.0386 0.1816 1>, 1=<rgba 0.175 0 0.1339 1>, 2=?, 3=?, Index=$Particles.PartitionParticles.Partition, Num=::NewEnumerator0)
        2=<rgba 0.135 0.0261 0.1226 1>
        3=<rgba 0.2 0.0386 0.1816 1>
        Index=$Particles.SpawnGroup
        Num=::NewEnumerator2
       )
       Variant 5 Color=&MergeHelper_ColorSwitch(
        0=&MergeHelper_ColorSwitch(0=<rgba 0.1177 0 0.379 1>, 1=<rgba 0.1814 0.0045 0.488 1>, 2=?, 3=?, Index=$Particles.PartitionParticles.Partition, Num=::NewEnumerator0)
        1=&MergeHelper_ColorSwitch(0=<rgba 0.1385 0 0.382 1>, 1=<rgba 0.1219 0 0.475 1>, 2=?, 3=?, Index=$Particles.PartitionParticles.Partition, Num=::NewEnumerator0)
        2=<rgba 0.1329 0 0.491 1>
        3=<rgba 0.1559 0.0037 0.552 1>
        Index=$Particles.SpawnGroup
        Num=::NewEnumerator2
       )
      )
      Float=&Multiply_Float(
       A=$System.TODScalar
       B=&FN_PickFloatDependOnPlatform(
        TargetPlatformActive=&FN_FrontEndFloatModifier(Gameplay=1, Front End=1)
        TargetPlatformNoActive=&FN_FrontEndFloatModifier(Gameplay=0.075, Front End=0.075)
       )
      )
     ))
    ScaleA=true
    ScaleRGB=true
   )
   &SpriteFacingAndAlignment(
    Alignment Coordinate Space=::Local
    Sprite Alignment=&RotateVectorByQuaternion(
     Vector To Rotate=<0 0 1>
     Quaternion=&MakeQuaternion(
      Quaternion From=0
      Axis=<0 1 0>
      Angle=&MergeHelper_FloatSwitch(
       0=&MergeHelper_FloatSwitch(0=-17, 1=-6, 2=?, 3=?, Index=$Particles.PartitionParticles.Partition, Num=::NewEnumerator0)
       1=&MergeHelper_FloatSwitch(0=90, 1=35, 2=?, 3=?, Index=$Particles.PartitionParticles.Partition, Num=::NewEnumerator0)
       2=120
       3=0
       Index=$Particles.SpawnGroup
       Num=::NewEnumerator2
      )
      From=?
      To=?
      Stabilization Mode:ENiagara_OrientVectorCount=::NewEnumerator0
      Side Direction:Vector3f=?
      Up Direction:Vector3f=?
      Angle Type=::NewEnumerator0
      Normalize:NiagaraBool=true
     )
    )
    WriteAlignment=true
    WriteFacing=false
   )
  ]
  rendererClasses [&NiagaraSpriteRendererProperties]
  renderers [{rendererClass=&NiagaraSpriteRendererProperties}]
 }}
summary {
 systemName=NS_Pickaxe_VerTet_IdleGlow
 emitters [{emitterName=CapsuleGlow, rendererClasses [&NiagaraSpriteRendererProperties]}]
 userVariables [
  User.bIsFrontEnd {type=NiagaraBool}
  User.bIsFrontEndPreview {type=NiagaraBool}
  User.Variant {type=NiagaraFloat, defaultValue=0, description=""}
 ]
}
