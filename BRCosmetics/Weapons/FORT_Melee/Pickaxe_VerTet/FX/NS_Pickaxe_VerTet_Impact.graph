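# Niagara impact effect for the VerTet harvesting pickaxe cosmetic, aimed along a system direction vector and rescaled for front-end display.
# Emitters shown include circuit sprites and mesh, surface-dependent debris with limited collision, and layered flashes with a light, flares and flying particles.
# Spawn groups drive per-element lifetime, motion and size within shared emitters.
# User.Variant selects color palettes and enables extra bursts; some flash colors also follow time of day.
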NiagaraSystem /BRCosmetics/Weapons/FORT_Melee/Pickaxe_VerTet/FX/NS_Pickaxe_VerTet_Impact
dependencies {
 usedDataInterfaces [
  &NiagaraDataInterfaceCurve
  &NiagaraDataInterfaceColorCurve
  &NiagaraDataInterfacePlatformSet
  &NiagaraDataInterfaceSpriteRendererInfo
  &NiagaraDataInterfaceVector2DCurve
  &NiagaraDataInterfaceArrayFloat3
 ]
 usedDynamicInputs [/Game/Effects/Niagara/Modules/Utility/FN_DirectionVectorByPickaxeType]
 usedModules [
  &self:SystemSpawnScript.SystemScriptSource.SystemScriptGraph.NiagaraNodeAssignment_1.SetVariables_#1
  &self:SystemSpawnScript.SystemScriptSource.SystemScriptGraph.NiagaraNodeAssignment_2.SetVariables_#2
  &self:SystemSpawnScript.SystemScriptSource.SystemScriptGraph.NiagaraNodeAssignment_0.SetVariables_#3
  &SystemState
  &FN_VariantManager
 ]
}
emitters {
 Circuit {
  emitterUpdateScript [
   &EmitterState()
   &SpawnBurst_Instantaneous(Use Spawn Probability=false, UseLoopCountLimit=false)
   &SpawnBurst_Instantaneous(Spawn Group=1, Use Spawn Probability=false, UseLoopCountLimit=false) {module=SpawnBurst_Instantaneous001}
   /SetVariables(Emitter.CircuitCoreColor:LinearColor=&VariantColorSelector(
     Variant Amount=4
     Original Color=<rgba 0.3953 0.2113 0.7043 1>
     Variant 1 Color=<rgba 0.51 0.5837 0.85 1>
     Variant 2 Color=<rgba 2 1.065 0.13 1>
     Variant 3 Color=<rgba 0.3 1.785 3 1>
     Variant 4 Color=<rgba 0.3333 0.2 1 1>
     Variant 5 Color=<rgba 0.1691 0.4522 0.4097 1>
    )) {isSetParameters=true}
  ]
  particleSpawnScript [
   /SetVariables(Particles.VisibilityTag=$Emitter.SpawnGroup) {isSetParameters=true}
   &InitializeParticle(
    Color=&VariantColorSelector(
     Variant Amount=4
     Original Color=<rgba 3.7333 1.6 8 5>
     Variant 1 Color=<rgba 3.25 3.9542 6.5 5>
     Variant 2 Color=<rgba 12 8.5 3.6 1>
     Variant 3 Color=<rgba 3.6 9.34 12 1>
     Variant 4 Color=<rgba 3.0667 1.2 4 5>
     Variant 5 Color=<rgba 3.32 0.8 8 5>
    )
    Lifetime=&MergeHelper_FloatSwitch(0=0.465, 1=0.18, 2=?, 3=?, Index=$Particles.VisibilityTag, Num=::NewEnumerator0)
    Mesh Scale Mode=::NewEnumerator3
    Mesh Uniform Scale=&Multiply_Float(A=0.7, B=&FN_FrontEndFloatModifier(Gameplay=1, Front End=0.5))
    Sprite Size Mode=::NewEnumerator3
    Sprite UV Mode=::NewEnumerator1
    Uniform Sprite Size=&Multiply_Float(A=70, B=&FN_FrontEndFloatModifier(Gameplay=1, Front End=0.5))
    UsePositionOffset=false
    Write Material Random=true
   )
   &InitialMeshOrientation(
    Mesh Orientation Mode=::NewEnumerator7
    Rotation=&MakeVector(X=&RandomRangeFloat(Minimum=0, Maximum=1), Y=0, Z=0)
    Rotation Coordinate Space=::NewEnumerator0
    Use Rotation Vector=true
   )
   &AddVelocity(
    Velocity=$System.DirectionVector
    Velocity Speed Scale=&MergeHelper_FloatSwitch(0=0, 1=100, 2=?, 3=?, Index=$Particles.VisibilityTag, Num=::NewEnumerator0)
   )
   &CameraOffset(Camera Offset Amount=15)
  ]
  particleUpdateScript [
   &ParticleState()
   &ScaleMeshSize(Scale Factor=&VectorFromFloat(Value=&FloatFromCurve(
      FloatCurve=DataInterface() {propertyValues {Curve {keys [{arriveTangent=1.3008, interpMode=RCIM_Cubic, leaveTangent=1.3008, tangentMode=RCTM_User, time=0, value=0.3}, <at 1 1.2 RCIM_Cubic>]}}}
      Scale Curve=0.9
     )))
   &ScaleSpriteSize(Uniform Curve Sprite Scale=DataInterface() {propertyValues {Curve {keys [{tangentMode=RCTM_User, time=0, value=1}, {tangentMode=RCTM_User, time=1, value=1.5}]}}})
   &DynamicMaterialParameters(
    Index 0 Param 1=&FloatFromCurve(FloatCurve=DataInterface() {propertyValues {Curve {keys [{arriveTangent=3.28, interpMode=RCIM_Cubic, leaveTangent=3.28, tangentMode=RCTM_User, time=0.365, value=-0.25}, <at 1 0.7>]}}})
    Index 0 Param 2=&FloatFromCurve(FloatCurve=DataInterface() {propertyValues {Curve {keys [{arriveTangent=3.0546, interpMode=RCIM_Cubic, leaveTangent=3.0546, tangentMode=RCTM_User, time=0, value=1}, <at 0.7 2>]}}})
    Module.Index0Float0Write=true
    Module.Index0Float1Write=true
    Module.Index0Float2Write=true
    Module.Index0Float3Write=true
   ) {module=DynamicMaterialParameters003}
   &FN_Sprite_Attributes(
    Color=&MultiplyLinearColor(
     Linear Color=&VariantColorSelector(
      Variant Amount=4
      Original Color=<rgba 0.1305 0 0.522 0.7>
      Variant 1 Color=<rgba 0.3132 0.3306 0.522 0.25>
      Variant 2 Color=<rgba 1 0.4702 0.065 0.1>
      Variant 3 Color=<rgba 0.0325 0.2896 0.5 0.4>
      Variant 4 Color=<rgba 0.087 0 0.522 0.5>
      Variant 5 Color=&ColorFromCurve(ColorCurve=DataInterface() {propertyValues {
         AlphaCurve {keys [<at 0.3171 0.7 RCIM_Cubic>,<at 0.7745 0.2 RCIM_Cubic>]}
         BlueCurve {keys [<at 0.3184 0.535 RCIM_Cubic>,<at 1 0.85 RCIM_Cubic>]}
         GreenCurve {keys [<at 0.3184 0 RCIM_Cubic>,<at 1 1 RCIM_Cubic>]}
         RedCurve {keys [<at 0.3184 0.1872 RCIM_Cubic>,<at 1 0 RCIM_Cubic>]}
        }})
     )
     Scale Factor=&MakeVector4FromVectorAndFloat(
      "Vector (RGB)"=<1 1 1>
      "Float (Alpha)"=&FloatFromCurve(FloatCurve=DataInterface() {propertyValues {Curve {keys [<at 0.5 1 RCIM_Cubic>,<at 1 0 RCIM_Cubic>]}}})
     )
    )
    Create Color Attribute=true
    Create Dynamic Material Parameter Attributes=true
    Create SpriteSize Attribute=true
    Index 0 Param 1=2
    Index 0 Param 2=1
    Index 0 Param 3=1
    Index 0 Param 4=1
    Module Output:ENiagara_AttributeSamplingApplyOutput=::NewEnumerator1
    Module.Index0Float0Write=true
    Module.Index0Float1Write=true
    Module.Index0Float2Write=true
    Module.Index0Float3Write=true
    SpriteSize=&Multiply_Vector2D(A=$Particles.SpriteSize, B=&Vector2DFromFloat(Value=2))
   )
   &ScaleColor(
    Scale RGB=&VectorFromFloat(Value=&FloatFromCurve(FloatCurve=DataInterface() {propertyValues {Curve {keys [<at 0 1.5>,<at 1 1>]}}}))
    ScaleA=true
    ScaleRGB=true
   )
   &DynamicMaterialParameters(
    Dynamic Parameter Output:ENiagara_AttributeSamplingApplyOutput=::NewEnumerator1
    Index 0 Param 1=&FloatFromCurve(FloatCurve=DataInterface() {propertyValues {Curve {keys [<at 0 1>,<at 0.2 1.375>,<at 1 0>]}}}, Scale Curve=1.2)
    Index 0 Param 2=1.3
    Index 0 Param 3=&FloatFromCurve(FloatCurve=DataInterface() {propertyValues {Curve {keys [<at 0.1 1>,<at 1 0>]}}})
    Index 0 Param 4=&FloatFromCurve(FloatCurve=DataInterface() {propertyValues {Curve {keys [
         <at 0 1>
         {arriveTangent=0.3079, interpMode=RCIM_Cubic, leaveTangent=0.3079, tangentMode=RCTM_User, time=0.3, value=0.9}
         <at 1 1 RCIM_Cubic>
        ]}}})
    Index 1 Param 1=100
    Index 1 Param 2=100
    Module.Index0Float0Write=true
    Module.Index0Float1Write=true
    Module.Index0Float2Write=true
    Module.Index0Float3Write=true
    Module.Index1Float0Write=true
    Module.Index1Float1Write=true
    Module.Index1Float2Write=true
    Module.Index1Float3Write=true
    Write Parameter Index 1=::NewEnumerator1
   ) {module=DynamicMaterialParameters001}
   &FN_Mesh_Attributes(
    Color=&VariantColorSelector(
     Variant Amount=4
     Original Color=&ColorFromCurve(ColorCurve=DataInterface() {propertyValues {
        AlphaCurve {keys [<at 0.5 2 RCIM_Cubic>,<at 1 0 RCIM_Cubic>]}
        BlueCurve {keys [<at 0 4 RCIM_Cubic>,<at 1 3 RCIM_Cubic>]}
        GreenCurve {keys [<at 0 0.4 RCIM_Cubic>,<at 1 0 RCIM_Cubic>]}
        RedCurve {keys [<at 0 1.6 RCIM_Cubic>,<at 1 0.5 RCIM_Cubic>]}
       }})
     Variant 1 Color=&ColorFromCurve(ColorCurve=DataInterface() {propertyValues {
        AlphaCurve {keys [<at 0.5 2 RCIM_Cubic>,<at 1 0 RCIM_Cubic>]}
        BlueCurve {keys [<at 0 2 RCIM_Cubic>,<at 1 1 RCIM_Cubic>]}
        GreenCurve {keys [<at 0 1.06 RCIM_Cubic>,<at 1 0.6867 RCIM_Cubic>]}
        RedCurve {keys [<at 0 0.8 RCIM_Cubic>,<at 1 0.6 RCIM_Cubic>]}
       }})
     Variant 2 Color=&ColorFromCurve(ColorCurve=DataInterface() {propertyValues {
        AlphaCurve {keys [<at 0.5 2 RCIM_Cubic>,<at 1 0 RCIM_Cubic>]}
        BlueCurve {keys [<at 0 0.78 RCIM_Cubic>,<at 1 0.2 RCIM_Cubic>]}
        GreenCurve {keys [<at 0 1.872 RCIM_Cubic>,<at 1 0.98 RCIM_Cubic>]}
        RedCurve {keys [<at 0 2.6 RCIM_Cubic>,<at 1 2 RCIM_Cubic>]}
       }})
     Variant 3 Color=&ColorFromCurve(ColorCurve=DataInterface() {propertyValues {
        AlphaCurve {keys [<at 0.5 2 RCIM_Cubic>,<at 1 0 RCIM_Cubic>]}
        BlueCurve {keys [<at 0 2 RCIM_Cubic>,<at 1 1 RCIM_Cubic>]}
        GreenCurve {keys [<at 0 1.5567 RCIM_Cubic>,<at 1 0.64 RCIM_Cubic>]}
        RedCurve {keys [<at 0 0.6 RCIM_Cubic>,<at 1 0.2 RCIM_Cubic>]}
       }})
     Variant 4 Color=&ColorFromCurve(ColorCurve=DataInterface() {propertyValues {
        AlphaCurve {keys [<at 0.5 2 RCIM_Cubic>,<at 1 0 RCIM_Cubic>]}
        BlueCurve {keys [<at 0.3468 3 RCIM_Cubic>,<at 1 2 RCIM_Cubic>]}
        GreenCurve {keys [<at 0.3468 0.345 RCIM_Cubic>,<at 1 0.2 RCIM_Cubic>]}
        RedCurve {keys [<at 0.3468 1.5397 RCIM_Cubic>,<at 1 0.5 RCIM_Cubic>]}
       }})
     Variant 5 Color=&ColorFromCurve(ColorCurve=DataInterface() {propertyValues {
        AlphaCurve {keys [<at 0.5 2 RCIM_Cubic>,<at 1 0 RCIM_Cubic>]}
        BlueCurve {keys [<at 0.3184 4 RCIM_Cubic>,<at 1 2.6 RCIM_Cubic>]}
        GreenCurve {keys [<at 0.3184 0.2 RCIM_Cubic>,<at 1 3 RCIM_Cubic>]}
        RedCurve {keys [<at 0.3184 1.5933 RCIM_Cubic>,<at 1 0 RCIM_Cubic>]}
       }})
    )
    Create Color Attribute=true
    Module Output=::NewEnumerator1
   )
   &DynamicMaterialParameters(
    Dynamic Parameter Output:ENiagara_AttributeSamplingApplyOutput=::NewEnumerator1
    Index 0 Param 1=1
    Index 0 Param 2=&FloatFromCurve(FloatCurve=DataInterface() {propertyValues {Curve {keys [<at 0 0>,<at 1 1>]}}})
    Module.Index0Float0Write=true
    Module.Index0Float1Write=true
    Module.Index0Float2Write=true
    Module.Index0Float3Write=true
   ) {module=DynamicMaterialParameters002}
   &FN_Sprite_Attributes(
    Color=&MultiplyLinearColor(
     Linear Color=&VariantColorSelector(
      Variant Amount=4
      Original Color=<rgba 0.7 0.3 1.5 5>
      Variant 1 Color=<rgba 0.6 0.6867 1 5>
      Variant 2 Color=<rgba 1 0.7083 0.3 1>
      Variant 3 Color=<rgba 0.3 0.7783 1 1>
      Variant 4 Color=<rgba 0.6174 0.2348 1 5>
      Variant 5 Color=&ColorFromCurve(ColorCurve=DataInterface() {propertyValues {
         AlphaCurve {keys [<at 0 1 RCIM_Cubic>]}
         BlueCurve {keys [<at 0.3171 2.5 RCIM_Cubic>,<at 1 1.7626 RCIM_Cubic>]}
         GreenCurve {keys [<at 0.3171 0.3261 RCIM_Cubic>,<at 1 2 RCIM_Cubic>]}
         RedCurve {keys [<at 0.3171 1.087 RCIM_Cubic>,<at 1 0.4174 RCIM_Cubic>]}
        }})
     )
     Scale Factor=&MakeVector4FromVectorAndFloat(
      "Vector (RGB)"=<1 1 1>
      "Float (Alpha)"=&FloatFromCurve(FloatCurve=DataInterface() {propertyValues {Curve {keys [{arriveTangent=-2.1568, interpMode=RCIM_Cubic, leaveTangent=-2.1568, tangentMode=RCTM_User, time=0, value=1}, <at 1 0>]}}})
     )
    )
    Create Color Attribute=true
    Create SpriteSize Attribute=true
    Module Output:ENiagara_AttributeSamplingApplyOutput=::NewEnumerator1
    SpriteSize=&Multiply_Vector2D_ByFloat(Float=$System.FrontEndScale, Vector2D=<35 100>)
   ) {module=FN_Sprite_Attributes001}
   &SolveForcesAndVelocity(Clamp Velocity=false, Limit Acceleration=false)
  ]
  rendererClasses [&NiagaraSpriteRendererProperties, &NiagaraMeshRendererProperties]
  renderers [
   {rendererClass=&NiagaraSpriteRendererProperties}
   {rendererClass=&NiagaraSpriteRendererProperties, rendererIndex=1}
   {rendererClass=&NiagaraMeshRendererProperties, rendererIndex=2}
   {rendererClass=&NiagaraSpriteRendererProperties, rendererIndex=3}
  ]
 }
 Directional_FN__PhysMatDebris {
  emitterUpdateScript [
   &EmitterState()
   &SpawnBurst_Instantaneous(
    Spawn Count=&RandomRangeInt(Minimum=5, Maximum=10)
    Use Spawn Probability=false
    UseLoopCountLimit=false
   )
  ]
  particleSpawnScript [
   /SetVariables(Particles.BiasedRandom=&FloatFromCurve(
     FloatCurve=DataInterface() {propertyValues {Curve {keys [
         <at 0.2 1 RCIM_Cubic>
         {arriveTangent=-3.3394, interpMode=RCIM_Cubic, leaveTangent=-3.3394, tangentMode=RCTM_User, time=0.5, value=0.5}
         <at 0.8 0 RCIM_Cubic>
        ]}}}
     CurveIndex=&RandomRangeFloat(Minimum=0, Maximum=1)
    )) {isSetParameters=true}
   &InitializeParticle(
    Color Mode=::NewEnumerator0
    Lifetime=&RandomRangeFloat(Minimum=0.5, Maximum=1.25)
    Mass=&Lerp_Float(A=0.6, B=1.2, Alpha=$Particles.BiasedRandom)
    Mass Mode=::NewEnumerator1
    Mesh Scale=&Multiply_VectorByFloat(
     Vector=&RandomRangeVector(Minimum=<0.9 0.4 0.5>, Maximum=<1 1 1>)
     Float=&Lerp_Float(A=0.7, B=1, Alpha=$Particles.BiasedRandom)
    )
    Mesh Scale Mode=::NewEnumerator4
    UsePositionOffset=false
    Write Material Random=true
   )
   &FN_PhysicalSurface(
    Dirt Mesh Count=2
    Glass Mesh Count=3
    Ice Mesh Count=2
    Metal Gravity=&RandomRangeFloat(Minimum=-350, Maximum=-1800)
    Metal Mesh Count=2
    Metal Velocity=&RandomRangeFloat(Minimum=80, Maximum=375)
    Metal Velocity Cone Angle=130
    Physical Surface=$User.PhysicalSurface
    Stone Gravity=&RandomRangeFloat(Minimum=-700, Maximum=-1250)
    Stone Mesh Count=2
    Stone Velocity=&RandomRangeFloat(Minimum=80, Maximum=550)
    Stone Velocity Cone Angle=120
    Wood Gravity=&RandomRangeFloat(Minimum=-450, Maximum=-900)
    Wood Mesh Count=2
    Wood Scale=1.5
    Wood Velocity=&RandomRangeFloat(Minimum=80, Maximum=600)
    Wood Velocity Cone Angle=130
   )
   &AddVelocityInCone(
    Cone Angle=$Output.FN_PhysicalSurface.SurfaceResponseConeAngle
    Cone Axis=$System.DirectionVector
    Module.Use Velocity Falloff On Cone Axis=false
    Velocity Strength=$Output.FN_PhysicalSurface.SurfaceResponseVelocity
   )
   &InitialMeshOrientation(Use Rotation Vector=false)
   &AddRotationalVelocity(Rotation Rate=&RandomVector(Vector Scale=&RandomRangeFloat(Minimum=1, Maximum=2)))
   &ApplyInitialForces()
   /SetVariables(Particles.GravityOverrideValue=$Output.FN_PhysicalSurface.GravityOverride) {isSetParameters=true}
  ]
  particleUpdateScript [
   &ParticleState()
   &GravityForce(Gravity=$Particles.GravityOverrideValue)
   &Collision(
    Advanced Aging Rate=1
    Age Colliding Particles=false
    CPU Collision Trace Channel=ECC_WorldStatic
    Collision Enabled=&Boolean_AndOperation(
     Bool1=&IsPlatformSetActive(New Platform Set=DataInterface() {propertyValues {Platforms {qualityLevelMask=-8}}})
     Bool2=&Boolean_AndOperation(
      Bool1=&SetBoolByFloatComparison(A=$Particles.Initial.Lifetime, B=1)
      Bool2=&SetBoolByFloatComparison(A=$Particles.Age, B=0.12)
     )
    )
    Control Roll On Collision=true
    Friction=1
    Friction During a Bounce=0.25
    Max CPU Trace Length=100
    Module.Control Rest State Via Surface Normal=false
    Module.EnableMaxCollisionCount=false
    Module.Randomize Collision Normal Vector=true
    Randomize Collision Normal=0.25
    ScaleRateOfRotation=false
   )
   &SolveForcesAndVelocity(Clamp Velocity=false, Limit Acceleration=false)
   &ScaleMeshSize(Scale Factor=&VectorFromFloat(Value=&FloatFromCurve(FloatCurve=DataInterface() {propertyValues {Curve {keys [<at 0.6 1 RCIM_Cubic>, {arriveTangent=-6.308, interpMode=RCIM_Cubic, leaveTangent=-6.308, tangentMode=RCTM_User, time=1, value=0}]}}})))
   &DynamicMaterialParameters(Module.Param0WriteEnabled=true, Module.Param1WriteEnabled=true, Module.Param2WriteEnabled=false, Module.Param3WriteEnabled=false)
  ]
  rendererClasses [&NiagaraSpriteRendererProperties, &NiagaraMeshRendererProperties]
  renderers [
   {rendererClass=&NiagaraSpriteRendererProperties}
   {rendererClass=&NiagaraMeshRendererProperties, rendererIndex=1}
   {rendererClass=&NiagaraMeshRendererProperties, rendererIndex=2}
   {rendererClass=&NiagaraMeshRendererProperties, rendererIndex=3}
   {rendererClass=&NiagaraMeshRendererProperties, rendererIndex=4}
   {rendererClass=&NiagaraMeshRendererProperties, rendererIndex=5}
   {rendererClass=&NiagaraMeshRendererProperties, rendererIndex=6}
   {rendererClass=&NiagaraMeshRendererProperties, rendererIndex=7}
   {rendererClass=&NiagaraMeshRendererProperties, rendererIndex=8}
  ]
 }
 FlashElements {
  emitterUpdateScript [
   &EmitterState()
   &SpawnBurst_Instantaneous(
    Spawn Count=&RandomRangeInt(Minimum=6, Maximum=8)
    Use Spawn Probability=false
    UseLoopCountLimit=false
   )
   &SpawnBurst_Instantaneous(Spawn Group=1, Use Spawn Probability=false, UseLoopCountLimit=false) {module=SpawnBurst_Instantaneous001}
   &SpawnBurst_Instantaneous(
    Spawn Group=2
    Spawn Probability=&MakeCustomFloatFromBool(
     Bool=&SetBoolByFloatComparison(A=$User.Variant, B=4, Comparison Type=::NewEnumerator4)
     True Float=1
     False Float=0
    )
    Use Spawn Probability=true
    UseLoopCountLimit=false
   ) {module=SpawnBurst_Instantaneous002}
  ]
  particleSpawnScript [
   /SetVariables(Particles.SpriteSize=<1 1>, Particles.VisibilityTag=&SpawnGroup()) {isSetParameters=true}
   &InitializeParticle(
    Color Mode=::NewEnumerator0
    Lifetime=&MergeHelper_FloatSwitch(0=0.25, 1=0.25, 2=0.3, 3=?, Index=$Particles.VisibilityTag, Num=::NewEnumerator1)
    Sprite Rotation Mode=::NewEnumerator2
    UsePositionOffset=false
    Write Material Random=true
   )
   &ShapeLocation(MaskSpawnGroup=false, Override Seed=false, Sphere Radius=1, UseOwnerScale=true)
   &SpriteFacingAndAlignment(Sprite Alignment=$Particles.ShapeLocation.ShapeNormal, WriteAlignment=true, WriteFacing=false)
   &FN_PerParticleLOD(Enable SpriteSize Distance Scaling=true, Sprite Size Scale Far=<2.5 2.5>)
   /SetVariables(
    Particles.Flash Spike Lifetime:NiagaraFloat=&RandomRangeFloat(Minimum=0.05, Maximum=0.07)
    Particles.Flash Spike Sprite Size:Vector2f=&Multiply_Vector2D_ByFloat(Float=$System.FrontEndScale, Vector2D=&RandomRangeVector2D(Minimum=<3 55>, Maximum=<10 70>))
   ) {isSetParameters=true}
   /SetVariables(
    Particles.Flash Star Cam Offset=15
    Particles.Flash Star Color:LinearColor=$System.TOD Color
    Particles.Flash Star Lifetime=0.15
    Particles.Flash Star Rotation:NiagaraFloat=&RandomRangeFloat(Minimum=0, Maximum=360)
    Particles.Flash Star Sprite Size:Vector2f=&Multiply_Vector2D_ByFloat(
     Float=&Multiply_Float(A=$System.FrontEndScale, B=&RandomRangeFloat(Minimum=120, Maximum=150))
     Vector2D=$Particles.SpriteSize
    )
    Particles.Flash Star SubUV Index:NiagaraFloat=&RandomRangeFloat(Minimum=0, Maximum=3)
   ) {isSetParameters=true}
   /SetVariables(
    Particles.Flash Flare Lifetime=0.15
    Particles.Flash Flare Sprite Size:Vector2f=&Multiply_Vector2D_ByFloat(Float=$System.FrontEndScale, Vector2D=<180 180>)
   ) {isSetParameters=true}
   &DynamicMaterialParameters(
    Index 0 Param 1=0.2
    Index 0 Param 2=1
    Module.Index0Param0Write=true
    Module.Index0Param1Write=true
    Module.Index0Param2Write=true
    Module.Index0Param3Write=true
   )
   &CameraOffset(Camera Offset Amount=15)
  ]
  particleUpdateScript [
   &ParticleState()
   /SetVariables(
    Particles.Flash Flare Age:NiagaraFloat=&Divide_Float(A=$Particles.Age, B=$Particles.Initial.Flash Flare Lifetime)
    Particles.Flash Spike Age:NiagaraFloat=&Divide_Float(A=$Particles.Age, B=$Particles.Initial.Flash Spike Lifetime)
    Particles.Flash Star Age:NiagaraFloat=&Divide_Float(A=$Particles.Age, B=$Particles.Initial.Flash Star Lifetime)
   ) {isSetParameters=true}
   /SetVariables(
    Particles.Flash Flare Color:LinearColor=&MakeLinearColorFromVectorAndFloat(
     "Vector (RGB)"=&Multiply_Vector(
      A=&Multiply_Vector(
       A=&MakeVectorFromLinearColorRGB(Color=&LerpLinearColors(StartColor=<rgba 1 1 1 1>, EndColor=$System.TOD Color, LerpFactor=0.5))
       B=&MakeVectorFromLinearColorRGB(Color=/Game/Effects/Niagara/DynamicInputs/TimeOfDay/DaySequenceTimeOfDayColor(
         Linear Color:LinearColor=&VariantColorSelector(
          Variant Amount=4
          Original Color=&ColorFromCurve(
           ColorCurve=DataInterface() {propertyValues {
             AlphaCurve {keys [<at 0 1 RCIM_Cubic>,<at 1 0 RCIM_Cubic>]}
             BlueCurve {keys [<at 0.0198 2>,<at 0.491 2>,<at 1 0.3663>]}
             GreenCurve {keys [<at 0.0198 0.25>,<at 0.491 0.2>,<at 1 0.0092>]}
             RedCurve {keys [<at 0.0198 0.6739>,<at 0.491 0.8716>,<at 1 0.049>]}
            }}
           CurveIndex=$Particles.Flash Flare Age
          )
          Variant 1 Color=&ColorFromCurve(
           ColorCurve=DataInterface() {propertyValues {
             AlphaCurve {keys [<at 0 1 RCIM_Cubic>,<at 1 0 RCIM_Cubic>]}
             BlueCurve {keys [<at 0.0198 2>,<at 0.491 1>,<at 1 0.3663>]}
             GreenCurve {keys [<at 0.0198 0.9033>,<at 0.491 0.6867>,<at 1 0.2802>]}
             RedCurve {keys [<at 0.0198 0.6>,<at 0.491 0.6>,<at 1 0.2564>]}
            }}
           CurveIndex=$Particles.Flash Flare Age
          )
          Variant 2 Color=&ColorFromCurve(
           ColorCurve=DataInterface() {propertyValues {
             AlphaCurve {keys [<at 0 1 RCIM_Cubic>,<at 1 0 RCIM_Cubic>]}
             BlueCurve {keys [<at 0.0198 0.2275>,<at 0.491 0.065>,<at 1 0.0357>]}
             GreenCurve {keys [<at 0.0198 0.5819>,<at 0.491 0.6104>,<at 1 0.2757>]}
             RedCurve {keys [<at 0.0198 0.7>,<at 0.491 1>,<at 1 0.55>]}
            }}
           CurveIndex=$Particles.Flash Flare Age
          )
          Variant 3 Color=&ColorFromCurve(
           ColorCurve=DataInterface() {propertyValues {
             AlphaCurve {keys [<at 0 1 RCIM_Cubic>,<at 1 0 RCIM_Cubic>]}
             BlueCurve {keys [<at 0.0198 2>,<at 0.491 2>,<at 1 0.366>]}
             GreenCurve {keys [<at 0.0198 1.4458>,<at 0.491 1.19>,<at 1 0.0865>]}
             RedCurve {keys [<at 0.0198 0.25>,<at 0.491 0.2>,<at 1 0.0091>]}
            }}
           CurveIndex=$Particles.Flash Flare Age
          )
          Variant 4 Color=&ColorFromCurve(
           ColorCurve=DataInterface() {propertyValues {
             AlphaCurve {keys [<at 0 1 RCIM_Cubic>,<at 1 0 RCIM_Cubic>]}
             BlueCurve {keys [<at 0.0198 2>,<at 0.491 2>,<at 1 0.3663>]}
             GreenCurve {keys [<at 0.0198 0.25>,<at 0.491 0.2>,<at 1 0.0092>]}
             RedCurve {keys [<at 0.0198 2>,<at 0.491 1.1>,<at 1 0.1282>]}
            }}
           CurveIndex=$Particles.Flash Flare Age
          )
          Variant 5 Color=&ColorFromCurve(
           ColorCurve=DataInterface() {propertyValues {
             AlphaCurve {keys [<at 0 1 RCIM_Cubic>,<at 1 0 RCIM_Cubic>]}
             BlueCurve {keys [<at 0.0198 2>,<at 0.491 2>,<at 1 0.3663>]}
             GreenCurve {keys [<at 0.0198 0.13>,<at 0.491 0.1>,<at 1 0.0092>]}
             RedCurve {keys [<at 0.0198 0.7845>,<at 0.491 0.67>,<at 1 0.0984>]}
            }}
           CurveIndex=$Particles.Flash Flare Age
          )
         )
         Time Of Day Influence:NiagaraFloat=1
        ))
      )
      B=&VectorFromFloat(Value=&FN_FrontEndFloatModifier(Gameplay=1.5, Front End=1))
     )
     "Float (Alpha)"=&FloatFromCurve(
      FloatCurve=DataInterface() {propertyValues {Curve {keys [{arriveTangent=-1.9727, interpMode=RCIM_Cubic, time=0.1, value=1}, <at 1 0 RCIM_Cubic>]}}}
      CurveIndex=$Particles.Flash Flare Age
     )
    )
    Particles.Flash Spike Color:LinearColor=&MakeLinearColorFromVectorAndFloat(
     "Vector (RGB)"=&MakeVectorFromLinearColorRGB(Color=&MultiplyLinearColor(
       Linear Color=&LerpLinearColors(StartColor=<rgba 1 1 1 1>, EndColor=$System.TOD Color, LerpFactor=0.5)
       Scale Factor=&MakeVector4_FromLinearColor(LinearColor=&VariantColorSelector(
         Variant Amount=4
         Original Color=<rgba 0.0942 0.0226 0.452 1>
         Variant 1 Color=<rgba 0.3164 0.3458 0.452 1>
         Variant 2 Color=<rgba 0.5 0.2351 0.0325 1>
         Variant 3 Color=<rgba 0.05 0.2 0.5 1>
         Variant 4 Color=<rgba 0.214 0.021 0.6 1>
         Variant 5 Color=<rgba 0.1371 0.0226 0.452 1>
        ))
      ))
     "Float (Alpha)"=&FloatFromCurve(FloatCurve=DataInterface() {propertyValues {Curve {keys [<at 0.7 1>,<at 1 0>]}}}, CurveIndex=$Particles.Flash Spike Age)
    )
    Particles.Flash Spike Sprite Size:Vector2f=&Multiply_Vector2D_ByFloat(
     Float=&FloatFromCurve(FloatCurve=DataInterface() {propertyValues {Curve {keys [<at 0 0.6>,<at 1 2>]}}}, CurveIndex=$Particles.Flash Spike Age)
     Vector2D=$Particles.Initial.Flash Spike Sprite Size
    )
    Particles.Flash Star Color:LinearColor=&MultiplyLinearColor(
     Linear Color=&LerpLinearColors(StartColor=<rgba 1 1 1 1>, EndColor=$System.TOD Color, LerpFactor=0.5)
     Scale Factor=&MakeVector4_FromLinearColor(LinearColor=&VariantColorSelector(
       Variant Amount=4
       Original Color=&ColorFromCurve(
        ColorCurve=DataInterface() {propertyValues {
          AlphaCurve {keys [<at 0.2935 1>,<at 0.9862 0>]}
          BlueCurve {keys [<at 0.0344 3>,<at 0.4768 2>]}
          GreenCurve {keys [<at 0.0344 0.15>,<at 0.4768 0.03>]}
          RedCurve {keys [<at 0.0344 1.5327>,<at 0.4768 0.2005>]}
         }}
        CurveIndex=$Particles.Flash Star Age
       )
       Variant 1 Color=&ColorFromCurve(
        ColorCurve=DataInterface() {propertyValues {
          AlphaCurve {keys [<at 0.2935 1>,<at 0.9862 0>]}
          BlueCurve {keys [<at 0.0344 3>,<at 0.4768 1>]}
          GreenCurve {keys [<at 0.0344 1.12>,<at 0.4768 0.6867>]}
          RedCurve {keys [<at 0.0344 0.6>,<at 0.4768 0.6>]}
         }}
        CurveIndex=$Particles.Flash Star Age
       )
       Variant 2 Color=&ColorFromCurve(
        ColorCurve=DataInterface() {propertyValues {
          AlphaCurve {keys [<at 0.2935 1>,<at 0.9862 0>]}
          BlueCurve {keys [<at 0.0344 0.12>,<at 0.4768 0.07>]}
          GreenCurve {keys [<at 0.0344 0.288>,<at 0.4768 0.343>]}
          RedCurve {keys [<at 0.0344 0.4>,<at 0.4768 0.7>]}
         }}
        CurveIndex=$Particles.Flash Star Age
       )
       Variant 3 Color=&ColorFromCurve(
        ColorCurve=DataInterface() {propertyValues {
          AlphaCurve {keys [<at 0.2935 1>,<at 0.9862 0>]}
          BlueCurve {keys [<at 0.0344 3>,<at 0.4768 2>]}
          GreenCurve {keys [<at 0.0344 2.0975>,<at 0.4768 0.4568>]}
          RedCurve {keys [<at 0.0344 0.15>,<at 0.4768 0.03>]}
         }}
        CurveIndex=$Particles.Flash Star Age
       )
       Variant 4 Color=&ColorFromCurve(
        ColorCurve=DataInterface() {propertyValues {
          AlphaCurve {keys [<at 0.2935 1>,<at 0.9862 0>]}
          BlueCurve {keys [<at 0.0344 4>,<at 0.4768 3>]}
          GreenCurve {keys [<at 0.0344 0.2>,<at 0.4768 0.045>]}
          RedCurve {keys [<at 0.0344 4>,<at 0.4768 0.2913>]}
         }}
        CurveIndex=$Particles.Flash Star Age
       )
       Variant 5 Color=&ColorFromCurve(
        ColorCurve=DataInterface() {propertyValues {
          AlphaCurve {keys [<at 0.2935 1>,<at 0.9862 0>]}
          BlueCurve {keys [<at 0.0344 3>,<at 0.4768 2>]}
          GreenCurve {keys [<at 0.0344 0.15>,<at 0.4768 0.03>]}
          RedCurve {keys [<at 0.0344 1.2425>,<at 0.4768 0.5225>]}
         }}
        CurveIndex=$Particles.Flash Star Age
       )
      ))
    )
    Particles.Flash Star Sprite Size:Vector2f=&Multiply_Vector2D_ByFloat(
     Float=&FloatFromCurve(
      FloatCurve=DataInterface() {propertyValues {Curve {keys [
          {arriveTangent=2.7013, interpMode=RCIM_Cubic, leaveTangent=2.7013, tangentMode=RCTM_User, time=0, value=0}
          {interpMode=RCIM_Cubic, tangentMode=RCTM_User, time=1, value=1}
         ]}}}
      CurveIndex=$Particles.Flash Star Age
     )
     Vector2D=$Particles.Initial.Flash Star Sprite Size
    )
   ) {isSetParameters=true}
   &FN_Light_Attributes(
    Color=&VariantColorSelector(
     Variant Amount=4
     Original Color=<rgba 0.4167 0.1 2 1>
     Variant 1 Color=<rgba 0.6 0.6867 1 1>
     Variant 2 Color=<rgba 2 1.0027 0.13 1>
     Variant 3 Color=<rgba 0.13 1.2832 2 1>
     Variant 4 Color=<rgba 0.9867 0.1 2 1>
     Variant 5 Color=<rgba 0 1.3 1.105 1>
    )
    Create Color Attribute=true
    Create Exponent Attribute=true
    Create Position Attribute=true
    Create Radius Attribute=true
    Exponent=2
    Radius=&FN_FrontEndFloatModifier(Gameplay=90, Front End=0)
    Scale Color=&FloatFromCurve(
     FloatCurve=DataInterface() {propertyValues {Curve {keys [
         {arriveTangent=-0.5504, interpMode=RCIM_Cubic, leaveTangent=-0.5504, tangentMode=RCTM_User, time=0, value=1}
         {arriveTangent=-0.5426, interpMode=RCIM_Cubic, leaveTangent=-0.5426, tangentMode=RCTM_User, time=0.55, value=0}
        ]}}}
     Scale Curve=7
    )
   )
  ]
  rendererClasses [&NiagaraSpriteRendererProperties, &NiagaraLightRendererProperties]
  renderers [
   {rendererClass=&NiagaraSpriteRendererProperties}
   {rendererClass=&NiagaraSpriteRendererProperties, rendererIndex=1}
   {rendererClass=&NiagaraSpriteRendererProperties, rendererIndex=2}
   {rendererClass=&NiagaraLightRendererProperties, rendererIndex=3}
  ]
 }
 Flash_Flare {
  emitterUpdateScript [
   &EmitterState()
   &SpawnBurst_Instantaneous(Use Spawn Probability=false, UseLoopCountLimit=false)
   &SpawnBurst_Instantaneous(Spawn Count=2, Spawn Group=1, Use Spawn Probability=false, UseLoopCountLimit=false) {module=SpawnBurst_Instantaneous001}
   &SpawnBurst_Instantaneous(
    Spawn Count=&MakeIntFromFloat(Float=&FN_PickFloatDependOnPlatform(
      TargetPlatform=DataInterface() {propertyValues {Platforms {qualityLevelMask=-29}}}
      TargetPlatformActive=&MakeFloatFromInt(Int=12)
      TargetPlatformNoActive=&MakeFloatFromInt(Int=30)
     ))
    Spawn Group=2
    Use Spawn Probability=false
    UseLoopCountLimit=false
   ) {module=SpawnBurst_Instantaneous002}
   &SpawnBurst_Instantaneous(
    Spawn Count=&RandomRangeInt(Minimum=3, Maximum=5)
    Spawn Group=3
    Spawn Probability=&MakeCustomFloatFromBool(
     Bool=&Boolean_OrOperation(
      Bool1=&SetBoolByFloatComparison(A=$User.Variant, B=2, Comparison Type=::NewEnumerator4)
      Bool2=&SetBoolByFloatComparison(A=$User.Variant, B=4, Comparison Type=::NewEnumerator4)
     )
     True Float=1
     False Float=0
    )
    Use Spawn Probability=true
    UseLoopCountLimit=false
   ) {module=SpawnBurst_Instantaneous003}
   &SpawnBurst_Instantaneous(
    Spawn Count=&RandomRangeInt(Minimum=3, Maximum=5)
    Spawn Group=4
    Spawn Probability=&MakeCustomFloatFromBool(
     Bool=&SetBoolByFloatComparison(A=$User.Variant, B=3, Comparison Type=::NewEnumerator4)
     True Float=1
     False Float=0
    )
    Use Spawn Probability=true
    UseLoopCountLimit=false
   ) {module=SpawnBurst_Instantaneous004}
   &SpawnBurst_Instantaneous(
    Spawn Count=&RandomRangeInt(Minimum=3, Maximum=5)
    Spawn Group=5
    Spawn Probability=&MakeCustomFloatFromBool(
     Bool=&SetBoolByFloatComparison(A=$User.Variant, B=3, Comparison Type=::NewEnumerator4)
     True Float=1
     False Float=0
    )
    Use Spawn Probability=true
    UseLoopCountLimit=false
   ) {module=SpawnBurst_Instantaneous005}
  ]
  particleSpawnScript [
   &InitializeParticle(
    Color=&MergeHelper_ColorSwitch(
     0=&VariantColorSelector(
      Variant Amount=4
      Original Color=<rgba 0.6125 0.15 2 1>
      Variant 1 Color=<rgba 0.8 1.06 2 1>
      Variant 2 Color=<rgba 1.5 0.8455 0.0975 1>
      Variant 3 Color=<rgba 0.25 1.4875 2.5 1>
      Variant 4 Color=<rgba 1.2292 0.15 2 1>
      Variant 5 Color=<rgba 0.84 0.26 2 1>
     )
     1=&VariantColorSelector(
      Variant Amount=4
      Original Color=&RandomRangeLinearColor(Minimum=<rgba 0.3667 0.05 1 1>, Maximum=<rgba 0.2083 0.05 1 1>)
      Variant 1 Color=&RandomRangeLinearColor(Minimum=<rgba 0.6 0.6867 1 1>, Maximum=<rgba 0.5 0.5583 1 1>)
      Variant 2 Color=&RandomRangeLinearColor(Minimum=<rgba 1 0.6792 0.125 1>, Maximum=<rgba 1 0.4702 0.065 1>)
      Variant 3 Color=&RandomRangeLinearColor(Minimum=<rgba 0.05 0.6992 1 1>, Maximum=<rgba 0.065 0.5792 1 1>)
      Variant 4 Color=&RandomRangeLinearColor(Minimum=<rgba 1 0.05 0.8892 1>, Maximum=<rgba 0.3667 0.05 1 1>)
      Variant 5 Color=&RandomRangeLinearColor(Minimum=<rgba 0.35 0 1 1>, Maximum=<rgba 0 1 0.85 1>)
     )
     2=&HueShiftLinearColor(
      Hue=&RandomRangeFloat(Minimum=-0.035, Maximum=0.035)
      Saturation=1
      Value=1
      Linear Color=&VariantColorSelector(
       Variant Amount=4
       Original Color=&RandomRangeLinearColor(Minimum=<rgba 0.8 0.2 2 1>, Maximum=<rgba 0.2583 0.1 2 1>)
       Variant 1 Color=&RandomRangeLinearColor(Minimum=<rgba 0.8 1.06 2 1>, Maximum=<rgba 0.6 0.6333 1 1>)
       Variant 2 Color=&RandomRangeLinearColor(Minimum=<rgba 1 0.85 0.4 1>, Maximum=<rgba 1.3 0.832 0.13 1>)
       Variant 3 Color=&RandomRangeLinearColor(Minimum=<rgba 0.125 0.7229 1 1>, Maximum=<rgba 0.13 0.7735 1.3 1>)
       Variant 4 Color=&RandomRangeLinearColor(Minimum=<rgba 2 0.2 2 1>, Maximum=<rgba 0.5 0.2 2 1>)
       Variant 5 Color=&RandomRangeLinearColor(Minimum=<rgba 0.765 0.1 2 1>, Maximum=<rgba 0.1 2 1.715 1>)
      )
     )
     3=?
     Index=$Emitter.SpawnGroup
     Num=::NewEnumerator1
    )
    Lifetime=&MergeHelper_FloatSwitch(
     0=0.18
     1=&RandomRangeFloat(Minimum=0.1, Maximum=0.15)
     2=&RandomRangeFloat(Minimum=0.05, Maximum=0.45)
     3=&RandomRangeFloat(Minimum=0.45, Maximum=0.6)
     4=&RandomRangeFloat(Minimum=0.45, Maximum=0.6)
     5=&RandomRangeFloat(Minimum=0.55, Maximum=0.7)
     Index=$Emitter.SpawnGroup
     Num=::NewEnumerator4
    )
    Mesh Scale Mode=::NewEnumerator3
    Mesh Uniform Scale=&Multiply_Float(A=&RandomRangeFloat(Minimum=0.18, Maximum=0.25), B=$System.FrontendScale)
    Sprite Size=&MergeHelper_Vector2Switch(
     0=&FN_FrontEndVector2Modifier(Front End:Vector2f=<50 190>, Gameplay:Vector2f=<80 240>)
     1=&Multiply_Vector2D(
      A=&UniformRangedVector2D(Minimum=<10 40>, Maximum=<12.5 70>)
      B=&FN_FrontEndVector2Modifier(Front End:Vector2f=<0.5 0.7>, Gameplay:Vector2f=<1 1>)
     )
     2=&Multiply_Vector2D_ByFloat(
      Float=&Multiply_Float(
       A=&FN_FrontEndFloatModifier(Gameplay=1, Front End=0.8)
       B=&RandomRangeFloat(Minimum=4.5, Maximum=5)
      )
      Vector2D=<1 1.5>
     )
     3=?
     4=?
     Index=$Emitter.SpawnGroup
     Num=::NewEnumerator1
    )
    Sprite Size Mode=::NewEnumerator4
    Sprite UV Mode=::NewEnumerator1
    UsePositionOffset=false
    Write Material Random=true
   )
   &ShapeLocation(
    Cylinder Height=&MergeHelper_FloatSwitch(0=0, 1=0, 2=80, 3=80, 4=80, 5=80, Index=$Emitter.SpawnGroup, Num=::NewEnumerator4)
    Cylinder Height Midpoint=0
    Cylinder Radius=&MergeHelper_FloatSwitch(0=0, 1=0, 2=3, 3=3, 4=3, 5=3, Index=$Emitter.SpawnGroup, Num=::NewEnumerator4)
    MaskSpawnGroup=false
    Override Seed=false
    Rotation Axis=$System.DirectionVector
    Rotation Mode=::NewEnumerator1
    Shape Primitive=::NewEnumerator1
    Surface Only=false
    UseOwnerScale=true
   )
   &CameraOffset(Camera Offset Amount=&MergeHelper_FloatSwitch(0=30, 1=25, 2=20, 3=25, 4=20, 5=25, Index=$Emitter.SpawnGroup, Num=::NewEnumerator4))
   &SubUVAnimation(
    End Frame Range Override=3
    Evaluation Type=::NewEnumerator0
    "Sprite Renderer "=DataInterface() {propertyValues {SpriteRenderer=&self:Flash_Flare_1.NiagaraSpriteRendererProperties_0}}
    SubUV Animation Mode:ENiagara_SubUVLookupModeV2=::NewEnumerator2
    UseEndFrame=true
    UseStartFrame=true
   )
   &DynamicMaterialParameters(
    Index 0 Param 1=0.2
    Index 0 Param 2=1
    Index 3 Param 1=1.5
    Index 3 Param 2=1
    Module.Index0Float0Write=true
    Module.Index0Float1Write=true
    Module.Index0Float2Write=true
    Module.Index0Float3Write=true
    Module.Index3Float0Write=true
    Module.Index3Float1Write=true
    Module.Index3Float2Write=true
    Module.Index3Float3Write=true
    Write Parameter Index 3=::NewEnumerator1
   )
   /SetVariables(Particles.DynamicParameterAlt:Vector4f=&MakeVector4(X=1, Y=1, Z=1, W=1)) {isSetParameters=true}
   &AddVelocity(
    Module.Use Falloff Distance=false
    Override Seed=false
    Velocity Mode=::NewEnumerator1
    Velocity Speed=&Multiply_Float(
     A=&MergeHelper_FloatSwitch(
      0=0
      1=10
      2=&RandomRangeFloat(Minimum=225, Maximum=300)
      3=&RandomRangeFloat(Minimum=225, Maximum=300)
      4=&RandomRangeFloat(Minimum=165, Maximum=210)
      5=&RandomRangeFloat(Minimum=225, Maximum=300)
      Index=$Emitter.SpawnGroup
      Num=::NewEnumerator4
     )
     B=$System.FrontEndScale
    )
   ) {module=AddVelocity001}
   &AddVelocity(
    Rotation Coordinate Space=::Simulation
    Velocity=$System.DirectionVector
    Velocity Speed Scale=&MergeHelper_FloatSwitch(
     0=&Multiply_Float(A=&RandomRangeFloat(Minimum=200, Maximum=300), B=$System.FrontEndScale)
     1=&Multiply_Float(A=100, B=$System.FrontEndScale)
     2=&Multiply_Float(A=&RandomRangeFloat(Minimum=0, Maximum=450), B=$System.FrontEndScale)
     3=&Multiply_Float(A=&RandomRangeFloat(Minimum=0, Maximum=275), B=$System.FrontEndScale)
     4=&Multiply_Float(A=&RandomRangeFloat(Minimum=0, Maximum=450), B=$System.FrontEndScale)
     5=&Multiply_Float(A=&RandomRangeFloat(Minimum=0, Maximum=275), B=$System.FrontEndScale)
     Index=$Emitter.SpawnGroup
     Num=::NewEnumerator4
    )
   )
   /SetVariables() {isSetParameters=true}
   /SetVariables(Particles.VisibilityTag=$Emitter.SpawnGroup) {isSetParameters=true}
   /SetVariables(Particles.FN_Sprite_Attributes.SpriteRotation:NiagaraFloat=&RandomRangeFloat(Minimum=0, Maximum=360)) {isSetParameters=true}
  ]
  particleUpdateScript [
   &ParticleState()
   &AccelerationForce(Acceleration=&MergeHelper_VectorSwitch(0=<0 0 0>, 1=<0 0 0>, 2=<0 0 -500>, 3=<0 0 -333>, 4=<0 0 -666>, 5=<0 0 -500>, Index=$Particles.VisibilityTag, Num=::NewEnumerator4))
   &Drag(
    Drag=&MergeHelper_FloatSwitch(
     0=0
     1=0
     2=&RandomRangeFloat(Minimum=5, Maximum=8)
     3=&RandomRangeFloat(Minimum=5, Maximum=8)
     4=&RandomRangeFloat(Minimum=5, Maximum=8)
     5=&RandomRangeFloat(Minimum=5, Maximum=8)
     Index=$Particles.VisibilityTag
     Num=::NewEnumerator4
    )
    Rotational Drag:NiagaraFloat=&MergeHelper_FloatSwitch(0=0, 1=0, 2=1, 3=1, 4=1, 5=1, Index=$Particles.VisibilityTag, Num=::NewEnumerator4)
    Use Linear Drag=true
    Use Rotational Drag=true
   )
   &CurlNoiseForce(
    Bias Noise=false
    Noise Frequency=&MergeHelper_FloatSwitch(0=25, 1=25, 2=25, 3=25, 4=25, 5=25, Index=$Particles.VisibilityTag, Num=::NewEnumerator4)
    Noise Strength=&MergeHelper_FloatSwitch(0=0, 1=0, 2=0, 3=200, 4=0, 5=350, Index=$Particles.VisibilityTag, Num=::NewEnumerator4)
    Pan Noise=false
   )
   &SolveForcesAndVelocity(Clamp Velocity=false, Limit Acceleration=false)
   &ScaleSpriteSizeBySpeed(
    Max Scale Factor=&MergeHelper_Vector2Switch(0=<1 1>, 1=<1 1>, 2=<0.2 5>, 3=?, 4=?, Index=$Particles.VisibilityTag, Num=::NewEnumerator1)
    Min Scale Factor=&MergeHelper_Vector2Switch(0=<1 1>, 1=<1 1>, 2=<0.2 0.2>, 3=?, 4=?, Index=$Particles.VisibilityTag, Num=::NewEnumerator1)
   )
   &ScaleSpriteSize(
    Scale Factor=&MergeHelper_Vector2Switch(
     0=&Vector2DFromCurve(
      Vector2Curve=DataInterface() {propertyValues {
        XCurve {keys [{tangentMode=RCTM_User, time=0, value=0}, <at 0.2 1.075>, {tangentMode=RCTM_User, time=1, value=-0.1}]}
        YCurve {keys [{tangentMode=RCTM_User, time=0, value=0}, <at 0.1 1.025>, {tangentMode=RCTM_User, time=1, value=0.8}]}
       }}
      ScaleCurve=<1 1>
     )
     1=&Vector2DFromCurve(
      Vector2Curve=DataInterface() {propertyValues {
        XCurve {keys [<at 0 1 RCIM_Cubic>,<at 1 0 RCIM_Cubic>]}
        YCurve {keys [{arriveTangent=3.9291, interpMode=RCIM_Cubic, leaveTangent=3.9291, tangentMode=RCTM_User, time=0, value=0}, <at 1 2 RCIM_Cubic>]}
       }}
      ScaleCurve=&Multiply_Vector2D_ByFloat(Float=1, Vector2D=<1.5 1>)
     )
     2=&Vector2DFromFloat(Value=&FloatFromCurve(FloatCurve=DataInterface() {propertyValues {Curve {keys [{arriveTangent=-6.725, interpMode=RCIM_Cubic, leaveTangent=-6.725, tangentMode=RCTM_User, time=0.65, value=1}, <at 1 0 RCIM_Cubic>]}}}))
     3=?
     4=?
     Index=$Particles.VisibilityTag
     Num=::NewEnumerator1
    )
    Scale Sprite Size Mode=::NewEnumerator2
   )
   &ScaleColor(
    Scale Alpha=&MergeHelper_FloatSwitch(
     0=1
     1=&FloatFromCurve(
      FloatCurve=DataInterface() {propertyValues {Curve {keys [<at 0 1 RCIM_Cubic>,<at 1 0>]}}}
      Scale Curve=&FNTimeOfDayScalar(Night Scale=0.5, Morning Scale=1, Evening Scale=0.5)
     )
     2=1
     3=?
     Index=$Particles.VisibilityTag
     Num=::NewEnumerator1
    )
    Scale RGB=&MergeHelper_VectorSwitch(
     0=<1 1 1>
     1=&VectorFromFloat(Value=&FloatFromCurve(FloatCurve=DataInterface() {propertyValues {Curve {keys [{arriveTangent=-15.8795, interpMode=RCIM_Cubic, leaveTangent=-15.8795, tangentMode=RCTM_User, time=0, value=10}, <at 1 1 RCIM_Cubic>]}}}))
     2=&VectorFromFloat(Value=&FloatFromCurve(
       FloatCurve=DataInterface() {propertyValues {Curve {keys [
           {interpMode=RCIM_Cubic, tangentMode=RCTM_User, time=0, value=1}
           <at 0.3 0.2>
           {interpMode=RCIM_Cubic, tangentMode=RCTM_User, time=1, value=0}
          ]}}}
       Scale Curve=10
      ))
     3=?
     Index=$Particles.VisibilityTag
     Num=::NewEnumerator1
    )
    ScaleA=true
    ScaleRGB=true
   )
   /SetVariables(
    Particles.GlowColor=&MakeLinearColorFromVectorAndFloat(
     "Vector (RGB)"=&MakeVectorFromLinearColorRGB(Color=&VariantColorSelector(
       Variant Amount=4
       Original Color=<rgba 0.25 0 1 1>
       Variant 1 Color=<rgba 0.4522 0.5435 1 1>
       Variant 2 Color=<rgba 1 0.4702 0.065 1>
       Variant 3 Color=<rgba 0.1 0.55 1 1>
       Variant 4 Color=<rgba 0.575 0.1 2 1>
       Variant 5 Color=<rgba 0.2667 0 1 1>
      ))
     "Float (Alpha)"=&FloatFromCurve(FloatCurve=DataInterface() {propertyValues {Curve {keys [<at 0.5 1>,<at 1 0>]}}}, Scale Curve=0.1)
    )
    Particles.GlowSize=&Multiply_Vector2D(A=$Particles.SpriteSize, B=<8 2>)
   ) {isSetParameters=true}
   &FN_Sprite_Attributes(
    Color=&HueShiftLinearColor(
     Hue=&RandomRangeFloat(Minimum=-0.03, Maximum=0.03)
     Saturation=1
     Value=1
     Linear Color=&MaskLinearColor(
      Mask Source=1
      Mask Parameter=&MakeIntFromFloat(Float=$User.Variant)
      Mask=2
      Masked Linear Color=&RandomRangeLinearColor(Minimum=<rgba 1 0.65 0.125 1>, Maximum=<rgba 1 0.5325 0.065 1>)
      Passthrough Linear Color=&RandomRangeLinearColor(Minimum=<rgba 0.25 0.1 1 1>, Maximum=<rgba 1 0.1 1 1>)
     )
    )
    Create Color Attribute=true
    Create Sprite Rotation Attribute=true
    Create SpriteSize Attribute=true
    Module Output:ENiagara_AttributeSamplingApplyOutput=::NewEnumerator1
    SpriteRotation:NiagaraFloat=&Add_Float(
     A=&Multiply_Float(A=$Engine.DeltaTime, B=&RandomRangeFloat(Minimum=75, Maximum=120))
     B=$Particles.FN_Sprite_Attributes.SpriteRotation
    )
    SpriteSize=&Vector2DFromFloat(Value=&Multiply_Float(
      A=&Multiply_Float(
       A=&RandomRangeFloat(Minimum=10, Maximum=14)
       B=&FN_FrontEndFloatModifier(Gameplay=1, Front End=0.6)
      )
      B=&FloatFromCurve(
       FloatCurve=DataInterface() {propertyValues {Curve {keys [
           {arriveTangent=9.0725, interpMode=RCIM_Cubic, leaveTangent=3.7536, tangentMode=RCTM_Break, time=0, value=0}
           <at 0.5 1 RCIM_Cubic>
           {arriveTangent=-0.0405, interpMode=RCIM_Cubic, leaveTangent=-0.0405, tangentMode=RCTM_User, time=1, value=0}
          ]}}}
       Scale Curve=&MaskFloat(
        Mask Source=1
        Mask Parameter=&MakeIntFromFloat(Float=$User.Variant)
        Mask=2
        Masked Float=1
        Passthrough Float=1.5
       )
      )
     ))
   )
   &FN_Sprite_Attributes(
    Color=&MaskLinearColor(
     Mask Source=1
     Mask Parameter=&MakeIntFromFloat(Float=$User.Variant)
     Mask=2
     Masked Linear Color=<rgba 1 0.4702 0.065 0.02>
     Passthrough Linear Color=<rgba 0.2083 0.05 1 0.05>
    )
    Create Color Attribute=true
    Create SpriteSize Attribute=true
    Module Output:ENiagara_AttributeSamplingApplyOutput=::NewEnumerator1
    SpriteSize=&Multiply_Vector2D_ByFloat(Float=1.35, Vector2D=$Particles.FN_Sprite_Attributes.SpriteSize)
   ) {module=FN_Sprite_Attributes001}
   &FN_Mesh_Attributes(Color=<rgba 0.7 0.925 1 1>, Create Color Attribute=true, Module Output=::NewEnumerator1)
   &ScaleMeshSize(Scale Factor=&VectorFromFloat(Value=&FloatFromCurve(
      FloatCurve=DataInterface() {propertyValues {Curve {keys [<at 0 1 RCIM_Cubic>,<at 1 0 RCIM_Cubic>]}}}
      Scale Curve=&FN_FrontEndFloatModifier(Gameplay=2, Front End=1)
     )))
   &FN_Sprite_Attributes(
    Color=&RandomRangeLinearColor(Minimum=<rgba 0.7 0.905 1 1>, Maximum=<rgba 0.7 0.865 1 1>)
    Create Color Attribute=true
    Create Sprite Rotation Attribute=true
    Create SpriteSize Attribute=true
    Module Output:ENiagara_AttributeSamplingApplyOutput=::NewEnumerator1
    SpriteRotation:NiagaraFloat=&Add_Float(
     A=&RandomRangeFloat(Minimum=0, Maximum=360)
     B=&Multiply_Float(A=$Engine.DeltaTime, B=&RandomRangeFloat(Minimum=25, Maximum=75))
    )
    SpriteSize=&Vector2DFromFloat(Value=&Multiply_Float(
      A=&RandomRangeFloat(Minimum=10, Maximum=14)
      B=&FloatFromCurve(
       FloatCurve=DataInterface() {propertyValues {Curve {keys [
           {arriveTangent=9.0725, interpMode=RCIM_Cubic, leaveTangent=3.7536, tangentMode=RCTM_Break, time=0, value=0}
           <at 0.3 1 RCIM_Cubic>
           {arriveTangent=-0.0405, interpMode=RCIM_Cubic, leaveTangent=-0.0405, tangentMode=RCTM_User, time=1, value=0}
          ]}}}
       Scale Curve=&FN_FrontEndFloatModifier(Gameplay=1, Front End=0.5)
      )
     ))
   ) {module=FN_Sprite_Attributes002}
  ]
  rendererClasses [&NiagaraSpriteRendererProperties, &NiagaraMeshRendererProperties]
  renderers [
   {rendererClass=&NiagaraSpriteRendererProperties}
   {rendererClass=&NiagaraSpriteRendererProperties, rendererIndex=1}
   {rendererClass=&NiagaraSpriteRendererProperties, rendererIndex=2}
   {rendererClass=&NiagaraSpriteRendererProperties, rendererIndex=3}
   {rendererClass=&NiagaraSpriteRendererProperties, rendererIndex=4}
   {rendererClass=&NiagaraSpriteRendererProperties, rendererIndex=5}
   {rendererClass=&NiagaraMeshRendererProperties, rendererIndex=6}
   {rendererClass=&NiagaraSpriteRendererProperties, rendererIndex=7}
  ]
 }
 Sparks {
  emitterUpdateScript [
   &EmitterState()
   &SpawnBurst_Instantaneous(
    Spawn Count=&MakeIntFromFloat(Float=&FN_PickFloatDependOnPlatform(
      TargetPlatformActive=&MakeFloatFromInt(Int=&RandomRangeInt(Minimum=15, Maximum=20))
      TargetPlatformNoActive=&MakeFloatFromInt(Int=&RandomRangeInt(Minimum=7, Maximum=10))
     ))
    Use Spawn Probability=false
    UseLoopCountLimit=false
   )
   &SpawnBurst_Instantaneous(
    Spawn Count=&MakeIntFromFloat(Float=&FN_PickFloatDependOnPlatform(
      TargetPlatformActive=&MakeFloatFromInt(Int=&RandomRangeInt(Minimum=2, Maximum=3))
      TargetPlatformNoActive=&MakeFloatFromInt(Int=2)
     ))
    Spawn Group=1
    Use Spawn Probability=false
    UseLoopCountLimit=false
   ) {module=SpawnBurst_Instantaneous001}
  ]
  particleSpawnScript [
   /SetVariables(Particles.VisibilityTag=$Emitter.SpawnGroup) {isSetParameters=true}
   &InitializeParticle(
    AdjustAlpha=false
    AdjustHue=true
    AdjustSaturation=false
    AdjustValue=false
    Color=&VariantColorSelector(
     Variant Amount=4
     Original Color=&RandomRangeLinearColor(Minimum=<rgba 0.8 0.2 2 1>, Maximum=<rgba 0.2583 0.1 2 1>)
     Variant 1 Color=&RandomRangeLinearColor(Minimum=<rgba 0.8 1.06 2 1>, Maximum=<rgba 0.6 0.6333 1 1>)
     Variant 2 Color=&RandomRangeLinearColor(Minimum=<rgba 1 0.85 0.4 1>, Maximum=<rgba 1.3 0.832 0.13 1>)
     Variant 3 Color=&RandomRangeLinearColor(Minimum=<rgba 0.125 0.7229 1 1>, Maximum=<rgba 0.13 0.7735 1.3 1>)
     Variant 4 Color=&RandomRangeLinearColor(Minimum=<rgba 2 0.2 2 1>, Maximum=<rgba 0.5 0.2 2 1>)
     Variant 5 Color=&RandomRangeLinearColor(Minimum=<rgba 0.765 0.1 2 1>, Maximum=<rgba 0.1 2 1.715 1>)
    )
    Color Mode=::NewEnumerator3
    Hue Shift Range:Vector2f=<-0.035 0.035>
    Lifetime=&MergeHelper_FloatSwitch(
     0=&RandomRangeFloat(Minimum=0.05, Maximum=0.45)
     1=&RandomRangeFloat(Minimum=0.3, Maximum=0.4)
     2=?
     3=?
     Index=$Particles.VisibilityTag
     Num=::NewEnumerator0
    )
    Sprite Size Mode=::NewEnumerator3
    Uniform Sprite Size=&Multiply_Float(
     A=&FN_FrontEndFloatModifier(Gameplay=1, Front End=0.8)
     B=&RandomRangeFloat(Minimum=2, Maximum=3.5)
    )
    UsePositionOffset=false
    Write Material Random=true
   )
   &ShapeLocation(
    Hemisphere Distribution=<1 0.5>
    MaskSpawnGroup=false
    Override Seed=false
    Rotation Angle=90
    Rotation Angle Type:ENiagara_AngleInput=::NewEnumerator0
    Rotation Mode=::NewEnumerator1
    Sphere Radius=&MergeHelper_FloatSwitch(
     0=1
     1=&MakeCustomFloatFromBool(Bool=$User.bIsFrontEnd, True Float=5, False Float=10)
     2=?
     3=?
     Index=$Particles.VisibilityTag
     Num=::NewEnumerator0
    )
    UseOwnerScale=true
   )
   &CameraOffset(Camera Offset Amount=&MergeHelper_FloatSwitch(
     0=25
     1=&RandomRangeFloat(Minimum=40, Maximum=50)
     2=?
     3=?
     Index=$Particles.VisibilityTag
     Num=::NewEnumerator0
    ))
   &AddVelocity(
    Cone Angle=180
    Module.UseConeFalloff=false
    Module.UseVelDistribution=false
    Override Seed=false
    Velocity Mode=::NewEnumerator2
    Velocity Speed=&Multiply_Float(
     A=&MergeHelper_FloatSwitch(
      0=&RandomRangeFloat(Minimum=400, Maximum=750)
      1=10
      2=?
      3=?
      Index=$Particles.VisibilityTag
      Num=::NewEnumerator0
     )
     B=$System.FrontEndScale
    )
   )
  ]
  particleUpdateScript [
   &ParticleState()
   &AccelerationForce(Acceleration=&MergeHelper_VectorSwitch(
     0=&RandomRangeVector(Minimum=<0 0 -500>, Maximum=<0 0 -500>)
     1=<0 0 0>
     2=?
     3=?
     Index=$Particles.VisibilityTag
     Num=::NewEnumerator0
    ))
   &Drag(
    Drag=&MergeHelper_FloatSwitch(
     0=&RandomRangeFloat(Minimum=5, Maximum=8)
     1=0
     2=?
     3=?
     Index=$Particles.VisibilityTag
     Num=::NewEnumerator0
    )
    Use Linear Drag=true
    Use Rotational Drag=true
   )
   &SolveForcesAndVelocity(Clamp Velocity=false, Limit Acceleration=false)
   &ScaleSpriteSizeBySpeed(Max Scale Factor=<0.35 10>, Min Scale Factor=<0.35 1>)
   &ScaleSpriteSize(
    Uniform Curve Scale=&Multiply_Float(A=1, B=&FN_FrontEndFloatModifier(Gameplay=1, Front End=0.5))
    Uniform Curve Sprite Scale=DataInterface() {propertyValues {Curve {keys [{arriveTangent=-6.725, interpMode=RCIM_Cubic, leaveTangent=-6.725, tangentMode=RCTM_User, time=0.65, value=1}, <at 1 0 RCIM_Cubic>]}}}
   )
   &ScaleColor(
    Scale RGB=&VectorFromFloat(Value=&FloatFromCurve(
      FloatCurve=DataInterface() {propertyValues {Curve {keys [
          {interpMode=RCIM_Cubic, tangentMode=RCTM_User, time=0, value=1}
          <at 0.3 0.2>
          {interpMode=RCIM_Cubic, tangentMode=RCTM_User, time=1, value=0}
         ]}}}
      Scale Curve=10
     ))
    ScaleA=true
    ScaleRGB=true
   )
   /SetVariables(
    Particles.GlowColor=&MakeLinearColorFromVectorAndFloat(
     "Vector (RGB)"=&MakeVectorFromLinearColorRGB(Color=&VariantColorSelector(
       Variant Amount=4
       Original Color=<rgba 0.25 0 1 1>
       Variant 1 Color=<rgba 0.4522 0.5435 1 1>
       Variant 2 Color=<rgba 1 0.4702 0.065 1>
       Variant 3 Color=<rgba 0.1 0.55 1 1>
       Variant 4 Color=<rgba 0.575 0.1 2 1>
       Variant 5 Color=<rgba 0.2667 0 1 1>
      ))
     "Float (Alpha)"=&FloatFromCurve(FloatCurve=DataInterface() {propertyValues {Curve {keys [<at 0.5 1>,<at 1 0>]}}}, Scale Curve=0.08)
    )
    Particles.GlowSize=&Multiply_Vector2D(A=$Particles.SpriteSize, B=<8 2>)
   ) {isSetParameters=true}
   &FN_Sprite_Attributes(
    Color=&MultiplyLinearColor(
     Linear Color=&RandomRangeLinearColor(Minimum=<rgba 0.8 0.917 1 0.15>, Maximum=<rgba 0.76 1 1 0.2>)
     Scale Factor=&MakeVector4FromVectorAndFloat(
      "Vector (RGB)"=<1 1 1>
      "Float (Alpha)"=&FloatFromCurve(FloatCurve=DataInterface() {propertyValues {Curve {keys [
           {arriveTangent=2.4358, interpMode=RCIM_Cubic, leaveTangent=2.4358, tangentMode=RCTM_User, time=0, value=0}
           <at 0.25 0.5 RCIM_Cubic>
           <at 1 0 RCIM_Cubic>
          ]}}})
     )
    )
    Create Color Attribute=true
    Create SpriteSize Attribute=true
    Module Output:ENiagara_AttributeSamplingApplyOutput=::NewEnumerator1
    SpriteSize=&Multiply_Vector2D(
     A=&Vector2DFromFloat(Value=&FloatFromCurve(FloatCurve=DataInterface() {propertyValues {Curve {keys [
           {arriveTangent=1.7761, interpMode=RCIM_Cubic, leaveTangent=1.7761, tangentMode=RCTM_User, time=0, value=0}
           {interpMode=RCIM_Cubic, tangentMode=RCTM_User, time=1, value=1}
          ]}}}))
     B=&Vector2DFromFloat(Value=&Multiply_Float(
       A=&FN_FrontEndFloatModifier(Gameplay=1, Front End=0.6)
       B=&RandomRangeFloat(Minimum=100, Maximum=130)
      ))
    )
   )
   &DynamicMaterialParameters(
    Dynamic Parameter Output:ENiagara_AttributeSamplingApplyOutput=::NewEnumerator1
    Index 0 Param 1=&Add_Float(
     A=&RandomRangeFloat(Minimum=0, Maximum=1)
     B=&FloatFromCurve(
      FloatCurve=DataInterface() {propertyValues {Curve {keys [
          <at 0 0 RCIM_Cubic>
          {arriveTangent=1.5947, interpMode=RCIM_Cubic, leaveTangent=1.5947, tangentMode=RCTM_User, time=0.2557, value=0.5257}
          <at 1 1 RCIM_Cubic>
         ]}}}
      CurveIndex=$Particles.Age
      Scale Curve=0.25
     )
    )
    Module.Index0Float0Write=true
    Module.Index0Float1Write=true
    Module.Index0Float2Write=true
    Module.Index0Float3Write=true
   ) {module=DynamicMaterialParameters001}
  ]
  rendererClasses [&NiagaraSpriteRendererProperties]
  renderers [
   {rendererClass=&NiagaraSpriteRendererProperties}
   {rendererClass=&NiagaraSpriteRendererProperties, rendererIndex=1}
   {rendererClass=&NiagaraSpriteRendererProperties, rendererIndex=2}
   {rendererClass=&NiagaraSpriteRendererProperties, rendererIndex=3}
  ]
 }
}
summary {
 systemName=NS_Pickaxe_VerTet_Impact
 emitters [
  {emitterName=Circuit, rendererClasses [&NiagaraSpriteRendererProperties, &NiagaraMeshRendererProperties]}
  {emitterName=FlashElements, rendererClasses [&NiagaraSpriteRendererProperties, &NiagaraLightRendererProperties]}
  {
   emitterName=Directional_FN__PhysMatDebris
   rendererClasses [&NiagaraSpriteRendererProperties, &NiagaraMeshRendererProperties]
  }
  {emitterName=Sparks, rendererClasses [&NiagaraSpriteRendererProperties]}
  {emitterName=Flash_Flare, rendererClasses [&NiagaraSpriteRendererProperties, &NiagaraMeshRendererProperties]}
 ]
 userVariables [
  User.bIsFrontEnd {type=NiagaraBool}
  User.bIsFrontEndPreview {type=NiagaraBool}
  User.PhysicalSurface {type=NiagaraInt32}
  User.SwingSectionIndex {type=NiagaraInt32}
  User.Variant {type=NiagaraFloat, defaultValue=0, description=""}
 ]
}
