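# Niagara swing trail for the VerTet pickaxe, with color and particle looks that change per User.Variant.
# MainTrail and TipTrail are anim-trail ribbons using variant-selected colors, faded by swing fade and partly by time-of-day scalars.
# Variant_Particles samples a socket and spawns per unit moved, gating spawn groups by variant via spawn probability.
# Spawn group drives particle color, size, lifetime and motion across sprite, mesh and ribbon renderers, with front-end scale modifiers.
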NiagaraSystem /BRCosmetics/Weapons/FORT_Melee/Pickaxe_VerTet/FX/NS_Pickaxe_VerTet_Trail
dependencies {
 usedDataInterfaces [&NiagaraDataInterfaceCurve, &NiagaraDataInterfaceColorCurve, &NiagaraDataInterfaceSocketReader]
 usedDynamicInputs []
 usedModules [
  &self:SystemSpawnScript.SystemScriptSource.SystemScriptGraph.NiagaraNodeAssignment_1.SetVariables_#1
  &self:SystemSpawnScript.SystemScriptSource.SystemScriptGraph.NiagaraNodeAssignment_5.SetVariables_#2
  &self:SystemSpawnScript.SystemScriptSource.SystemScriptGraph.NiagaraNodeAssignment_0.SetVariables_#3
  &SystemState
  &self:SystemSpawnScript.SystemScriptSource.SystemScriptGraph.NiagaraNodeAssignment_4.SetVariables_#4
  &FN_VariantManager
 ]
}
emitters {
 MainTrail {
  emitterSpawnScript [&FN_RibbonIDOffsetUtility()]
  emitterUpdateScript [
   &EmitterState()
   &FN_SpawnAnimTrail(AnimTrailWidthMode=::FromSecond)
   /SetVariables(Emitter.EmitterRandom:NiagaraFloat=&RandomRangeFloat(Minimum=0, Maximum=1)) {isSetParameters=true}
  ]
  particleSpawnScript [
   &FN_InitializeAnimTrailBoneIndecies(First Bone or Socket Index=1, Second Bone or Socket Index=0)
   &FN_InitializeAnimTrail(AnimTrailWidth=1.2, Lifetime=0.22)
  ]
  particleUpdateScript [
   &FN_UpdateAnimTrail() {module=FN_UpdateAnimTrail001}
   &DynamicMaterialParameters(
    Index 0 Param 1=0.8
    Index 1 Param 3=&Add_Float(
     A=$Emitter.EmitterRandom
     B=&FloatFromCurve(FloatCurve=DataInterface() {propertyValues {Curve {keys [<at 0 0>,<at 1 -0.6>]}}}, CurveIndex=$Emitter.Age)
    )
    Index 1 Param 4=1
    Index 2 Param 1=20
    Index 2 Param 2=1.5
    Index 2 Param 4=0.6
    Module.Index0Float0Write=true
    Module.Index0Float1Write=true
    Module.Index0Float2Write=true
    Module.Index0Float3Write=true
    Module.Index1Float0Write=true
    Module.Index1Float1Write=true
    Module.Index1Float2Write=true
    Module.Index1Float3Write=true
    Module.Index2Float0Write=true
    Module.Index2Float1Write=true
    Module.Index2Float2Write=false
    Module.Index2Float3Write=false
    Write Parameter Index 1=::NewEnumerator1
    Write Parameter Index 2=::NewEnumerator1
   )
   &DynamicMaterialParameters(
    Dynamic Parameter Output:ENiagara_AttributeSamplingApplyOutput=::NewEnumerator1
    Index 0 Param 1=1
    Index 0 Param 3=1
    Index 0 Param 4=1
    Index 1 Param 3=&Add_Float(
     A=$Emitter.EmitterRandom
     B=&FloatFromCurve(FloatCurve=DataInterface() {propertyValues {Curve {keys [<at 0 0>,<at 1 -0.6>]}}}, CurveIndex=$Emitter.Age, Scale Curve=0.2)
    )
    Index 1 Param 4=1
    Index 2 Param 1=20
    Index 2 Param 2=4
    Index 2 Param 4=0.6
    Module.Index0Float0Write=true
    Module.Index0Float1Write=true
    Module.Index0Float2Write=true
    Module.Index0Float3Write=true
    Module.Index1Float0Write=true
    Module.Index1Float1Write=true
    Module.Index1Float2Write=true
    Module.Index1Float3Write=true
    Module.Index2Float0Write=true
    Module.Index2Float1Write=true
    Module.Index2Float2Write=false
    Module.Index2Float3Write=false
    Write Parameter Index 1=::NewEnumerator1
    Write Parameter Index 2=::NewEnumerator1
   ) {module=DynamicMaterialParameters001}
   &ScaleColor(
    Scale Alpha=&FloatFromCurve(
     FloatCurve=DataInterface() {propertyValues {Curve {keys [<at 0 1 RCIM_Cubic>,<at 1 0 RCIM_Cubic>]}}}
     Scale Curve=&Multiply_Float(
      A=&Multiply_Float(A=$System.SwingGeneralFade, B=0.6)
      B=&FNTimeOfDayScalar(Night Scale=0.9, Morning Scale=1, Evening Scale=1)
     )
    )
    Scale RGB=&MakeVectorFromLinearColorRGB(Color=&VariantColorSelector(
      Variant Amount=4
      Original Color=&ColorFromCurve(ColorCurve=DataInterface() {propertyValues {
         AlphaCurve {keys [<at 0 1 RCIM_Cubic>,<at 1 0 RCIM_Cubic>]}
         BlueCurve {keys [<at 0 0.535 RCIM_Cubic>,<at 0.5 0.7 RCIM_Cubic>]}
         GreenCurve {keys [<at 0 0.0062 RCIM_Cubic>,<at 0.5 0.0704 RCIM_Cubic>]}
         RedCurve {keys [<at 0 0.0625 RCIM_Cubic>,<at 0.5 0.3716 RCIM_Cubic>]}
        }})
      Variant 1 Color=&ColorFromCurve(ColorCurve=DataInterface() {propertyValues {
         AlphaCurve {keys [<at 0 1 RCIM_Cubic>,<at 1 0 RCIM_Cubic>]}
         BlueCurve {keys [<at 0 0.6 RCIM_Cubic>,<at 0.5 0.4 RCIM_Cubic>]}
         GreenCurve {keys [<at 0 0.412 RCIM_Cubic>,<at 0.5 0.306 RCIM_Cubic>]}
         RedCurve {keys [<at 0 0.36 RCIM_Cubic>,<at 0.5 0.28 RCIM_Cubic>]}
        }})
      Variant 2 Color=&ColorFromCurve(ColorCurve=DataInterface() {propertyValues {
         AlphaCurve {keys [<at 0 1 RCIM_Cubic>,<at 1 0 RCIM_Cubic>]}
         BlueCurve {keys [<at 0 0.15 RCIM_Cubic>,<at 0.5 0.11 RCIM_Cubic>]}
         GreenCurve {keys [<at 0 0.325 RCIM_Cubic>,<at 0.5 0.3007 RCIM_Cubic>]}
         RedCurve {keys [<at 0 0.5 RCIM_Cubic>,<at 0.5 0.55 RCIM_Cubic>]}
        }})
      Variant 3 Color=&ColorFromCurve(ColorCurve=DataInterface() {propertyValues {
         AlphaCurve {keys [<at 0 1 RCIM_Cubic>,<at 1 0 RCIM_Cubic>]}
         BlueCurve {keys [<at 0 0.5 RCIM_Cubic>,<at 0.5 0.35 RCIM_Cubic>]}
         GreenCurve {keys [<at 0 0.3454 RCIM_Cubic>,<at 0.5 0.1664 RCIM_Cubic>]}
         RedCurve {keys [<at 0 0.0108 RCIM_Cubic>,<at 0.5 0 RCIM_Cubic>]}
        }})
      Variant 4 Color=&ColorFromCurve(ColorCurve=DataInterface() {propertyValues {
         AlphaCurve {keys [<at 0 1 RCIM_Cubic>,<at 1 0 RCIM_Cubic>]}
         BlueCurve {keys [<at 0 0.5 RCIM_Cubic>,<at 0.5 0.5 RCIM_Cubic>]}
         GreenCurve {keys [<at 0 0.05 RCIM_Cubic>,<at 0.5 0.05 RCIM_Cubic>]}
         RedCurve {keys [<at 0 0.23 RCIM_Cubic>,<at 0.5 0.185 RCIM_Cubic>]}
        }})
      Variant 5 Color=&ColorFromCurve(ColorCurve=DataInterface() {propertyValues {
         AlphaCurve {keys [<at 0 1 RCIM_Cubic>,<at 1 0 RCIM_Cubic>]}
         BlueCurve {keys [<at 0 0.535 RCIM_Cubic>,<at 0.5 0.4 RCIM_Cubic>]}
         GreenCurve {keys [<at 0 0 RCIM_Cubic>,<at 0.5 0 RCIM_Cubic>]}
         RedCurve {keys [<at 0 0.1783 RCIM_Cubic>,<at 0.5 0.12 RCIM_Cubic>]}
        }})
     ))
    ScaleA=true
    ScaleRGB=true
   )
   &FN_Sprite_Attributes(
    Color=&MakeLinearColorFromVectorAndFloat(
     "Vector (RGB)"=&MakeVectorFromLinearColorRGB(Color=&VariantColorSelector(
       Variant Amount=4
       Original Color=&ColorFromCurve(ColorCurve=DataInterface() {propertyValues {
          AlphaCurve {keys [<at 0 1 RCIM_Cubic>,<at 1 0 RCIM_Cubic>]}
          BlueCurve {keys [<at 0 2 RCIM_Cubic>,<at 0.5 1.5 RCIM_Cubic>]}
          GreenCurve {keys [<at 0 0 RCIM_Cubic>,<at 0.5 0.075 RCIM_Cubic>]}
          RedCurve {keys [<at 0 0.2128 RCIM_Cubic>,<at 0.5 0.7566 RCIM_Cubic>]}
         }})
       Variant 1 Color=&ColorFromCurve(ColorCurve=DataInterface() {propertyValues {
          AlphaCurve {keys [<at 0 1 RCIM_Cubic>,<at 1 0 RCIM_Cubic>]}
          BlueCurve {keys [<at 0 0.53 RCIM_Cubic>,<at 0.5 0.375 RCIM_Cubic>]}
          GreenCurve {keys [<at 0 0.4054 RCIM_Cubic>,<at 0.5 0.3089 RCIM_Cubic>]}
          RedCurve {keys [<at 0 0.371 RCIM_Cubic>,<at 0.5 0.2906 RCIM_Cubic>]}
         }})
       Variant 2 Color=&ColorFromCurve(ColorCurve=DataInterface() {propertyValues {
          AlphaCurve {keys [<at 0 1 RCIM_Cubic>,<at 1 0 RCIM_Cubic>]}
          BlueCurve {keys [<at 0 0.128 RCIM_Cubic>,<at 0.5 0.0351 RCIM_Cubic>]}
          GreenCurve {keys [<at 0 0.4352 RCIM_Cubic>,<at 0.5 0.3044 RCIM_Cubic>]}
          RedCurve {keys [<at 0 0.64 RCIM_Cubic>,<at 0.5 0.54 RCIM_Cubic>]}
         }})
       Variant 3 Color=&ColorFromCurve(ColorCurve=DataInterface() {propertyValues {
          AlphaCurve {keys [<at 0 1 RCIM_Cubic>,<at 1 0 RCIM_Cubic>]}
          BlueCurve {keys [<at 0 0.735 RCIM_Cubic>,<at 0.5 0.5 RCIM_Cubic>]}
          GreenCurve {keys [<at 0 0.5174 RCIM_Cubic>,<at 0.5 0.2548 RCIM_Cubic>]}
          RedCurve {keys [<at 0 0.0478 RCIM_Cubic>,<at 0.5 0.0325 RCIM_Cubic>]}
         }})
       Variant 4 Color=&ColorFromCurve(ColorCurve=DataInterface() {propertyValues {
          AlphaCurve {keys [<at 0 1 RCIM_Cubic>,<at 1 0 RCIM_Cubic>]}
          BlueCurve {keys [<at 0 0.45 RCIM_Cubic>,<at 0.5 0.4 RCIM_Cubic>]}
          GreenCurve {keys [<at 0 0.0562 RCIM_Cubic>,<at 0.5 0.04 RCIM_Cubic>]}
          RedCurve {keys [<at 0 0.45 RCIM_Cubic>,<at 0.5 0.334 RCIM_Cubic>]}
         }})
       Variant 5 Color=&ColorFromCurve(ColorCurve=DataInterface() {propertyValues {
          AlphaCurve {keys [<at 0 1 RCIM_Cubic>,<at 1 0 RCIM_Cubic>]}
          BlueCurve {keys [<at 0 1.35 RCIM_Cubic>,<at 0.5 1 RCIM_Cubic>]}
          GreenCurve {keys [<at 0 0.1688 RCIM_Cubic>,<at 0.5 0.125 RCIM_Cubic>]}
          RedCurve {keys [<at 0 0.6609 RCIM_Cubic>,<at 0.5 0.3875 RCIM_Cubic>]}
         }})
      ))
     "Float (Alpha)"=&FloatFromCurve(
      FloatCurve=DataInterface() {propertyValues {Curve {keys [<at 0 1 RCIM_Cubic>,<at 1 0 RCIM_Cubic>]}}}
      Scale Curve=&Multiply_Float(A=$System.ToDScalar, B=$System.SwingGeneralFade)
     )
    )
    Create Color Attribute=true
    Module Output:ENiagara_AttributeSamplingApplyOutput=::NewEnumerator1
   )
  ]
  rendererClasses [&NiagaraRibbonRendererProperties]
  renderers [
   {rendererClass=&NiagaraRibbonRendererProperties}
   {rendererClass=&NiagaraRibbonRendererProperties, rendererIndex=1}
  ]
 }
 TipTrail {
  emitterSpawnScript [&FN_RibbonIDOffsetUtility()]
  emitterUpdateScript [
   &EmitterState()
   &FN_SpawnAnimTrail(AnimTrailWidthMode=::FromFirst)
  ]
  particleSpawnScript [
   &FN_InitializeAnimTrailBoneIndecies()
   &FN_InitializeAnimTrail(
    AnimTrailWidth=0.3
    Color=&VariantColorSelector(
     Variant Amount=4
     Original Color=<rgba 1.5312 0.375 5 1>
     Variant 1 Color=<rgba 0.8 1.06 2 1>
     Variant 2 Color=<rgba 0.7 0.5267 0.2275 1>
     Variant 3 Color=<rgba 0 0.956 1.5 1>
     Variant 4 Color=<rgba 0 1 0.8667 1>
     Variant 5 Color=<rgba 0 1 0.85 1>
    )
    Lifetime=0.225
   )
  ]
  particleUpdateScript [
   &FN_UpdateAnimTrail() {module=FN_UpdateAnimTrail001}
   &ScaleColor(
    Scale Alpha=$System.SwingGeneralFade
    Scale RGB=&VectorFromFloat(Value=&FloatFromCurve(FloatCurve=DataInterface() {propertyValues {Curve {keys [<at 0 3>,<at 0.2 1>,<at 1 0>]}}}, Scale Curve=$System.ToDScalar))
    ScaleA=true
    ScaleRGB=true
   ) {module=ScaleColor001}
   &ScaleRibbonWidth(Ribbon Width Scale=&FloatFromCurve(FloatCurve=DataInterface() {propertyValues {Curve {keys [<at 0 1>, {leaveTangent=-1.7704, time=1, value=0}]}}}))
   &DynamicMaterialParameters(
    Index 0 Param 4=1
    Index 1 Param 4=1
    Index 2 Param 1=40
    Index 2 Param 2=5
    Index 2 Param 4=0.6
    Module.Index0Float0Write=true
    Module.Index0Float1Write=true
    Module.Index0Float2Write=true
    Module.Index0Float3Write=true
    Module.Index1Float0Write=true
    Module.Index1Float1Write=true
    Module.Index1Float2Write=true
    Module.Index1Float3Write=true
    Module.Index2Float0Write=true
    Module.Index2Float1Write=true
    Module.Index2Float2Write=false
    Module.Index2Float3Write=false
    Write Parameter Index 1=::NewEnumerator1
    Write Parameter Index 2=::NewEnumerator1
   )
  ]
  rendererClasses [&NiagaraRibbonRendererProperties]
  renderers [{rendererClass=&NiagaraRibbonRendererProperties}]
 }
 Variant_Particles {
  emitterSpawnScript [/SetVariables(Emitter.isGalaxy:NiagaraBool=/Niagara/DynamicInputs/Helpers/MaskBool(
     Mask Source:NiagaraInt32=1
     Mask Parameter:NiagaraInt32=&MakeIntFromFloat(Float=$User.Variant)
     Mask:NiagaraInt32=4
     Masked Bool:NiagaraBool=true
     Passthrough Bool:NiagaraBool=false
    )) {isSetParameters=true}]
  emitterUpdateScript [
   &EmitterState(Life Cycle Mode=::NewEnumerator1, Loop Behavior=::NewEnumerator1, Loop Delay=0.05, Loop Duration=0.25, UseLoopDelay=true)
   &SpawnPerUnit(
    Module.Use Max Movement Threshold=false
    Module.Use Movement Tolerance=false
    Spawn Probability:NiagaraFloat=&MakeCustomFloatFromBool(
     Bool=&SetBoolByFloatComparison(A=$User.Variant, B=2, Comparison Type=::NewEnumerator4)
     True Float=1
     False Float=0
    )
    Use Spawn Probability=true
   )
   &SpawnPerUnit(
    Module.Use Max Movement Threshold=false
    Module.Use Movement Tolerance=false
    Spawn Group=1
    Spawn Probability:NiagaraFloat=&MakeCustomFloatFromBool(
     Bool=&SetBoolByFloatComparison(A=$User.Variant, B=3, Comparison Type=::NewEnumerator4)
     True Float=1
     False Float=0
    )
    Spawn Spacing=27
    Use Spawn Probability=true
   ) {module=SpawnPerUnit001}
   &SpawnPerUnit(
    Module.Use Max Movement Threshold=false
    Module.Use Movement Tolerance=false
    Spawn Group=2
    Spawn Probability:NiagaraFloat=&MakeCustomFloatFromBool(
     Bool=&SetBoolByFloatComparison(A=$User.Variant, B=3, Comparison Type=::NewEnumerator4)
     True Float=1
     False Float=0
    )
    Use Spawn Probability=true
   ) {module=SpawnPerUnit002}
   &SpawnPerUnit(
    Module.Use Max Movement Threshold=false
    Module.Use Movement Tolerance=false
    Spawn Group=3
    Spawn Probability:NiagaraFloat=&MakeCustomFloatFromBool(
     Bool=&SetBoolByFloatComparison(A=$User.Variant, B=5, Comparison Type=::NewEnumerator4)
     True Float=1
     False Float=0
    )
    Use Spawn Probability=true
   ) {module=SpawnPerUnit003}
   &SpawnPerUnit(
    Module.Use Max Movement Threshold=false
    Module.Use Movement Tolerance=false
    Spawn Group=4
    Spawn Probability:NiagaraFloat=&MakeCustomFloatFromBool(
     Bool=&SetBoolByFloatComparison(A=$User.Variant, B=4, Comparison Type=::NewEnumerator4)
     True Float=1
     False Float=0
    )
    Spawn Spacing=10
    Use Spawn Probability=true
   ) {module=SpawnPerUnit004}
  ]
  particleSpawnScript [
   &InitializeParticle(
    Color=&MergeHelper_ColorSwitch(
     0=&RandomRangeLinearColor(Minimum=<rgba 10 6.5 1.25 1>, Maximum=<rgba 10 5.325 0.65 1>)
     1=<rgba 0.7 0.925 1 1>
     2=&RandomRangeLinearColor(Minimum=<rgba 0.7 0.905 1 1>, Maximum=<rgba 0.7 0.865 1 1>)
     3=<rgba 0 0 0 0>
     4=<rgba 0 0 0 0>
     5=&Multiply_LinearColorByFloat(Linear Color=<rgba 1 1 1 1>, Float=&FN_FrontEndFloatModifier(Gameplay=1, Front End=1.3))
     Index=$Emitter.SpawnGroup
     Num=::NewEnumerator4
    )
    Lifetime=&MergeHelper_FloatSwitch(
     0=&RandomRangeFloat(Minimum=0.3, Maximum=0.45)
     1=&RandomRangeFloat(Minimum=0.3, Maximum=0.45)
     2=&RandomRangeFloat(Minimum=0.3, Maximum=0.45)
     3=&RandomRangeFloat(Minimum=0.3, Maximum=0.45)
     4=0.25
     5=&RandomRangeFloat(Minimum=0.45, Maximum=0.75)
     Index=$Emitter.SpawnGroup
     Num=::NewEnumerator4
    )
    Mesh Scale Mode=::NewEnumerator3
    Mesh Uniform Scale=&Multiply_Float(
     A=&RandomRangeFloat(Minimum=0.09, Maximum=0.14)
     B=&FN_FrontEndFloatModifier(Gameplay=1, Front End=0.6)
    )
    Ribbon Width=&Multiply_Float(
     A=&MaskFloat(Mask Source=1, Mask Parameter=$Emitter.SpawnGroup, Mask=5, Masked Float=0, Passthrough Float=0.75)
     B=&FN_FrontEndFloatModifier(Gameplay=1, Front End=0.8)
    )
    Ribbon Width Mode=::NewEnumerator1
    Sprite Rotation Mode=::NewEnumerator2
    Sprite Size Mode=::NewEnumerator3
    Uniform Sprite Size=&Multiply_Float(
     A=&MergeHelper_FloatSwitch(
      0=&RandomRangeFloat(Minimum=4.5, Maximum=8.5)
      1=&RandomRangeFloat(Minimum=6, Maximum=11)
      2=&RandomRangeFloat(Minimum=5, Maximum=9)
      3=&RandomRangeFloat(Minimum=6, Maximum=11)
      4=&RandomRangeFloat(Minimum=10, Maximum=15)
      Index=$Emitter.SpawnGroup
      Num=::NewEnumerator3
     )
     B=&FN_FrontEndFloatModifier(Gameplay=1, Front End=0.6)
    )
    UsePositionOffset=false
    Write Material Random=true
   )
   &SocketLocation(
    Interpolate Socket Sampling=true
    Orientation Sampling=::NewEnumerator1
    SamplePosOffset=true
    Sampled Position Offset=&MergeHelper_VectorSwitch(
     0=<0 0 0>
     1=<0 0 0>
     2=<0 15 0>
     3=<0 5 0>
     Index=&MakeIntFromFloat(Float=$User.Variant)
     Num=::NewEnumerator2
    )
    Socket Reader=$System.Socket Reader
    Socket Sampling Mode=::NewEnumerator3
    VelocitySpeedLimit=false
   )
   &ShapeLocation(MaskSpawnGroup=false, Override Seed=false, Shape Origin=$Output.SocketLocation.SampledPosition, Sphere Radius=10, UseOwnerScale=true)
   &AddVelocity(
    Module.Use Falloff Distance=false
    Override Seed=false
    Velocity Mode=::NewEnumerator1
    Velocity Origin=$Output.SocketLocation.SampledPosition
    Velocity Speed=&MaskFloat(
     Mask Source=1
     Mask Parameter=$Emitter.SpawnGroup
     Mask=5
     Masked Float=&RandomRangeFloat(Minimum=50, Maximum=120)
     Passthrough Float=&RandomRangeFloat(Minimum=0, Maximum=50)
    )
   )
   /SetVariables(Particles.VisibilityTag=$Emitter.SpawnGroup) {isSetParameters=true}
   /SetVariables(Particles.RibbonID=/Niagara/DynamicInputs/NiagaraID/MakeNiagaraIDFromInt(
     Index:NiagaraInt32=/Niagara/DynamicInputs/Divide/Divide_Int(A:NiagaraInt32=&ReturnParticleIDAttributes(), B:NiagaraInt32=36)
     Acquire Tag:NiagaraInt32=0
    )) {isSetParameters=true}
   &InheritVelocity(
    Inherited Velocity Amount Scale=&MaskVector(
     Mask Source=1
     Mask Parameter=$Emitter.SpawnGroup
     Mask=5
     Masked Vector=&VectorFromFloat(Value=&RandomRangeFloat(Minimum=0.1, Maximum=1))
     Passthrough Vector=<1 1 1>
    )
    Inherited Velocity Speed Limit:NiagaraFloat=&MergeHelper_FloatSwitch(
     0=0
     1=&RandomRangeFloat(Minimum=175, Maximum=300)
     2=75
     3=0
     4=0
     5=200
     Index=$Emitter.SpawnGroup
     Num=::NewEnumerator4
    )
    Speed Limit=true
    UseSpeedThreshold=false
    Velocity Source=$Output.SocketLocation.SampledVelocity
   )
  ]
  particleUpdateScript [
   &ParticleState()
   &ScaleMeshSize(Scale Factor=&VectorFromFloat(Value=&FloatFromCurve(
      FloatCurve=DataInterface() {propertyValues {Curve {keys [
          {interpMode=RCIM_Cubic, leaveTangentWeight=0.0667, tangentMode=RCTM_User, time=0, value=0}
          {arriveTangentWeight=0.0667, interpMode=RCIM_Cubic, leaveTangentWeight=0.2667, tangentMode=RCTM_User, time=0.2, value=1}
          {arriveTangentWeight=0.2667, interpMode=RCIM_Cubic, tangentMode=RCTM_User, time=1, value=0}
         ]}}}
      Scale Curve=&FN_FrontEndFloatModifier(Gameplay=2, Front End=1)
     )))
   &SpriteRotationRate(Rotation Rate=&FloatFromCurve(
     FloatCurve=DataInterface() {propertyValues {Curve {keys [{interpMode=RCIM_Cubic, tangentMode=RCTM_User, time=0, value=1}, {interpMode=RCIM_Cubic, tangentMode=RCTM_User, time=1, value=0}]}}}
     Scale Curve=&Multiply_Float(
      A=&MaskFloat(
       Mask Source=1
       Mask Parameter=$Particles.VisibilityTag
       Mask=5
       Masked Float=&Multiply_Float(
        A=&RandomRangeFloat(Minimum=1, Maximum=0.35)
        B=&FloatFromCurve(
         FloatCurve=DataInterface() {propertyValues {Curve {keys [{arriveTangent=-2, interpMode=RCIM_Cubic, time=0.15, value=1}, {interpMode=RCIM_Cubic, leaveTangent=-8.0089, time=0.5, value=0}]}}}
         Scale Curve=800
        )
       )
       Passthrough Float=&RandomRangeFloat(Minimum=100, Maximum=200)
      )
      B=&UniformAOrBFloat(A=-1, Evaluation Type=::NewEnumerator0, Recalculate Random Each Loop=true, Override Seed=false)
     )
    ))
   &Drag(
    Drag=&Multiply_Float(
     A=&UniformRangedFloat(Minimum=1.2, Maximum=1.5)
     B=&FloatFromCurve(
      FloatCurve=DataInterface() {propertyValues {Curve {keys [
          {arriveTangent=-5.3551, interpMode=RCIM_Cubic, leaveTangent=-5.3551, tangentMode=RCTM_User, time=0, value=0}
          {arriveTangent=-138.8873, interpMode=RCIM_Cubic, leaveTangent=-138.8873, tangentMode=RCTM_User, time=0.05, value=7}
          <at 0.2 0.5>
         ]}}}
      Scale Curve=4
     )
    )
    Use Linear Drag=true
    Use Rotational Drag=false
   )
   &AccelerationForce(Acceleration=&MaskVector(
     Mask Source=1
     Mask Parameter=$Particles.VisibilityTag
     Mask=5
     Masked Vector=&RandomRangeVector(Minimum=<0 0 -600>, Maximum=<0 0 -960>)
     Passthrough Vector=&RandomRangeVector(Minimum=<0 0 -150>, Maximum=<0 0 -200>)
    ))
   &CurlNoiseForce(
    Noise Frequency=30
    Noise Strength=&MergeHelper_FloatSwitch(0=200, 1=100, 2=200, 3=300, 4=200, Index=$Particles.VisibilityTag, Num=::NewEnumerator3)
    Pan Noise=false
   )
   &ScaleSpriteSize(
    Uniform Curve Scale=&FN_FrontEndFloatModifier(Gameplay=1, Front End=2)
    Uniform Curve Sprite Scale=DataInterface() {propertyValues {Curve {keys [
        {interpMode=RCIM_Cubic, leaveTangentWeight=0.0667, tangentMode=RCTM_User, time=0, value=0}
        {arriveTangentWeight=0.0667, interpMode=RCIM_Cubic, leaveTangentWeight=0.2667, tangentMode=RCTM_User, time=0.2, value=1}
        {arriveTangentWeight=0.5588, interpMode=RCIM_Cubic, tangentMode=RCTM_User, time=1, value=0}
       ]}}}
   )
   &ScaleColor(
    Scale Alpha=&Multiply_Float(
     A=&FloatFromCurve(FloatCurve=DataInterface() {propertyValues {Curve {keys [<at 0.3 1>,<at 1 0>]}}})
     B=&FNTimeOfDayScalar(Night Scale=0.5, Morning Scale=0.75, Evening Scale=0.75)
    )
    ScaleA=true
    ScaleRGB=true
   )
   /SetVariables(
    Particles.GlowColor=&MakeLinearColorFromVectorAndFloat(
     "Vector (RGB)"=&MakeVectorFromLinearColorRGB(Color=&MergeHelper_ColorSwitch(
       0=&Multiply_LinearColorByFloat(
        Linear Color=&ColorFromCurve(
         ColorCurve=DataInterface() {propertyValues {
           AlphaCurve {keys [<at 0 1 RCIM_Cubic>]}
           BlueCurve {keys [
             <at 0 1>
             <at 0.2788 1>
             {arriveTangent=0.4199, interpMode=RCIM_Cubic, leaveTangent=0.4199, time=0.5087, value=0.9509}
             <at 0.7781 1.3267>
             <at 1 1>
            ]}
           GreenCurve {keys [
             <at 0 0.6539>
             <at 0.2788 0.2522>
             {arriveTangent=1.0874, interpMode=RCIM_Cubic, leaveTangent=1.0874, time=0.5087, value=0.5156}
             <at 0.7781 1.5>
             <at 1 0.6539>
            ]}
           RedCurve {keys [
             <at 0 0.4301>
             <at 0.2788 0.613>
             {arriveTangent=-0.1655, interpMode=RCIM_Cubic, leaveTangent=-0.1655, time=0.5087, value=1}
             <at 0.7781 0.3013>
             <at 1 0.4301>
            ]}
          }}
         CurveIndex=&Frac_Float(Float=&Add_Float(A=$Particles.NormalizedAge, B=&RandomRangeFloat(Minimum=0, Maximum=-0.5)))
        )
        Float=&RandomRangeFloat(Minimum=0.2, Maximum=0.5)
       )
       1=&Multiply_LinearColorByFloat(
        Linear Color=&ColorFromCurve(
         ColorCurve=DataInterface() {propertyValues {
           AlphaCurve {keys [<at 0 1 RCIM_Cubic>]}
           BlueCurve {keys [
             <at 0.0284 0.8677>
             {arriveTangent=-0.4812, interpMode=RCIM_Cubic, leaveTangent=-0.4812, time=0.4737, value=0.8646}
             <at 0.7925 0.5>
             <at 0.9918 0.8646>
            ]}
           GreenCurve {keys [
             <at 0.0284 1>
             {arriveTangent=-1.2221, interpMode=RCIM_Cubic, leaveTangent=-1.2221, time=0.4737, value=0.1067}
             <at 0.7925 0.0662>
             <at 0.9918 0.7183>
            ]}
           RedCurve {keys [
             <at 0.0284 0>
             {arriveTangent=0.4161, interpMode=RCIM_Cubic, leaveTangent=0.4161, time=0.4737, value=0.231}
             <at 0.7925 0.3179>
             <at 0.9918 0>
            ]}
          }}
         CurveIndex=&Frac_Float(Float=&Add_Float(A=$Particles.NormalizedAge, B=&RandomRangeFloat(Minimum=0, Maximum=1)))
        )
        Float=&RandomRangeFloat(Minimum=0.2, Maximum=0.5)
       )
       2=?
       3=?
       Index=/Niagara/DynamicInputs/Helpers/MaskInt(
        Mask Source:NiagaraInt32=1
        Mask Parameter:NiagaraInt32=&MakeIntFromFloat(Float=$User.Variant)
        Mask:NiagaraInt32=5
        Masked Int:NiagaraInt32=0
        Passthrough Int:NiagaraInt32=1
       )
       Num=::NewEnumerator0
      ))
     "Float (Alpha)"=1
    )
    Particles.RibbonColor:LinearColor=&MakeLinearColorFromVectorAndFloat(
     "Vector (RGB)"=&MakeVectorFromLinearColorRGB(Color=&Multiply_LinearColorByFloat(
       Linear Color=&ColorFromCurve(
        ColorCurve=DataInterface() {propertyValues {
          AlphaCurve {keys [<at 0 1 RCIM_Cubic>]}
          BlueCurve {keys [<at 0.0378 0.8677>, {arriveTangent=-0.0033, interpMode=RCIM_Cubic, time=0.3642, value=0.8646}, <at 0.7623 0.5>, <at 0.9918 0.8646>]}
          GreenCurve {keys [<at 0.0378 1>, {arriveTangent=-0.2924, interpMode=RCIM_Cubic, time=0.3642, value=0.1067}, <at 0.7623 0.0662>, <at 0.9918 0.7183>]}
          RedCurve {keys [<at 0.0378 0>,<at 0.3642 0.231 RCIM_Cubic>,<at 0.7623 0.3179>,<at 0.9918 0>]}
         }}
        CurveIndex=&Frac_Float(Float=&Add_Float(
          A=&Multiply_Float(A=$Particles.NormalizedAge, B=0.3)
          B=&RandomRangeFloat(Minimum=0, Maximum=1)
         ))
       )
       Float=&RandomRangeFloat(Minimum=0.35, Maximum=0.6)
      ))
     "Float (Alpha)"=&FloatFromCurve(FloatCurve=DataInterface() {propertyValues {Curve {keys [<at 0.5 1>,<at 1 0>]}}}, Scale Curve=1.1)
    )
   ) {isSetParameters=true}
   &FN_Sprite_Attributes(
    Color=<rgba 1 0.4702 0.065 0.075>
    Create Color Attribute=true
    Create SpriteSize Attribute=true
    Module Output:ENiagara_AttributeSamplingApplyOutput=::NewEnumerator1
    SpriteSize=&Multiply_Vector2D_ByFloat(Float=2, Vector2D=$Particles.SpriteSize)
   ) {module=FN_Sprite_Attributes001}
   &SolveForcesAndVelocity(Clamp Velocity=false, Limit Acceleration=false)
  ]
  rendererClasses [
   &NiagaraSpriteRendererProperties
   &NiagaraRibbonRendererProperties
   &NiagaraMeshRendererProperties
  ]
  renderers [
   {rendererClass=&NiagaraSpriteRendererProperties}
   {rendererClass=&NiagaraSpriteRendererProperties, rendererIndex=1}
   {rendererClass=&NiagaraSpriteRendererProperties, rendererIndex=2}
   {rendererClass=&NiagaraRibbonRendererProperties, rendererIndex=3}
   {rendererClass=&NiagaraSpriteRendererProperties, rendererIndex=4}
   {rendererClass=&NiagaraMeshRendererProperties, rendererIndex=5}
   {rendererClass=&NiagaraSpriteRendererProperties, rendererIndex=6}
  ]
 }
}
summary {
 systemName=NS_Pickaxe_VerTet_Trail
 emitters [
  {emitterName=MainTrail, rendererClasses [&NiagaraRibbonRendererProperties]}
  {emitterName=TipTrail, rendererClasses [&NiagaraRibbonRendererProperties]}
  {
   emitterName=Variant_Particles
   rendererClasses [
    &NiagaraSpriteRendererProperties
    &NiagaraRibbonRendererProperties
    &NiagaraMeshRendererProperties
   ]
  }
 ]
 userVariables [
  User.bIsFrontEnd {type=NiagaraBool}
  User.bIsFrontEndPreview {type=NiagaraBool}
  User.RibbonIDOffset {type=NiagaraInt32}
  User.SwingSectionIndex {type=NiagaraInt32}
  User.Variant {type=NiagaraFloat, defaultValue=0, description=""}
 ]
}
